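# Material function that adds animated caustics to the emissive of the SaltLattice character body, scaled by SaltLattice_UseCaustics.
# Two caustic layers project on preskinned position, distorted by a panning mask, with camera-based parallax dropped at low quality.
# A chromatic curve atlas, tints and fresnel shape the color, plus a faint noise-driven sine shimmer.
# An optional TechArtMask limits the caustics and a BaseColor multiplier blend.
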MaterialFunction /BRCosmetics/Characters/Player/Female/Medium/Bodies/F_MED_SaltLattice/Materials/MF_SaltLattice_Caustics
graph {expressions {
  MaterialExpressionAdd_0 {
   class=MaterialExpressionAdd
   inputs [
    A {from=MaterialExpressionGetMaterialAttributes_0, output=EmissiveColor}
    B {from=MaterialExpressionMultiply_2}
   ]
  }
  MaterialExpressionAdd_11 {
   class=MaterialExpressionAdd
   inputs [
    A {from=MaterialExpressionMultiply_17}
    B {from=MaterialExpressionMultiply_18}
   ]
  }
  MaterialExpressionAdd_13 {
   class=MaterialExpressionAdd
   inputs [
    A {from=MaterialExpressionNoise_0}
    B {from=MaterialExpressionMultiply_21}
   ]
  }
  MaterialExpressionAdd_15 {
   class=MaterialExpressionAdd
   inputs [
    A {from=MaterialExpressionReroute_0}
    B {from=MaterialExpressionMultiply_25}
   ]
  }
  MaterialExpressionAdd_16 {
   class=MaterialExpressionAdd
   inputs [
    A {from=MaterialExpressionMultiply_0}
    B {from=MaterialExpressionMultiply_23}
   ]
   properties {constB=0}
  }
  MaterialExpressionAdd_17 {
   class=MaterialExpressionAdd
   inputs [
    A {from=MaterialExpressionShadingPathSwitch_0}
    B {from=MaterialExpressionMultiply_29}
   ]
  }
  MaterialExpressionAdd_2 {
   class=MaterialExpressionAdd
   inputs [
    A {from=MaterialExpressionPanner_3}
    B {from=MaterialExpressionConstantBiasScale_0}
   ]
  }
  MaterialExpressionAdd_20 {
   class=MaterialExpressionAdd
   inputs [
    A {from=MaterialExpressionMultiply_36}
    B {from=MaterialExpressionMultiply_5}
   ]
  }
  MaterialExpressionAdd_3 {
   class=MaterialExpressionAdd
   inputs [
    A {from=MaterialExpressionConvert_2}
    B {from=MaterialExpressionMultiply_8}
   ]
  }
  MaterialExpressionAdd_4 {
   class=MaterialExpressionAdd
   inputs [
    A {from=MaterialExpressionMultiply_4}
    B {from=MaterialExpressionSaturate_2}
   ]
  }
  MaterialExpressionAdd_6 {
   class=MaterialExpressionAdd
   inputs [
    A {from=MaterialExpressionConvert_0}
    B {from=MaterialExpressionConvert_1}
   ]
  }
  MaterialExpressionAdd_7 {
   class=MaterialExpressionAdd
   inputs [
    A {from=MaterialExpressionConvert_0}
    B {from=MaterialExpressionConvert_1}
   ]
  }
  MaterialExpressionAdd_8 {
   class=MaterialExpressionAdd
   inputs [
    A {from=MaterialExpressionReroute_0}
    B {from=MaterialExpressionMultiply_12}
   ]
  }
  MaterialExpressionAdd_9 {
   class=MaterialExpressionAdd
   inputs [
    A {from=MaterialExpressionShadingPathSwitch_1}
    B {from=MaterialExpressionMultiply_29}
   ]
  }
  MaterialExpressionAppendVector_0 {
   class=MaterialExpressionAppendVector
   inputs [
    A {from=MaterialExpressionConstant2Vector_0, output=""}
    B {from=MaterialExpressionScalarParameter_8, output=""}
   ]
  }
  MaterialExpressionCameraPositionWS_0 {class=MaterialExpressionCameraPositionWS, inputs [], properties {}}
  MaterialExpressionComponentMask_0 {class=MaterialExpressionComponentMask, inputs [None {from=MaterialExpressionFunctionInput_2}], properties {r=true}}
  MaterialExpressionConstant2Vector_0 {class=MaterialExpressionConstant2Vector, properties {g=1, r=1}}
  MaterialExpressionConstant2Vector_2 {class=MaterialExpressionConstant2Vector}
  MaterialExpressionConstantBiasScale_0 {class=MaterialExpressionConstantBiasScale, inputs [None {from=MaterialExpressionSine_0}], properties {scale=0.1}}
  MaterialExpressionConstantBiasScale_1 {class=MaterialExpressionConstantBiasScale, inputs [None {from=MaterialExpressionSine_1}]}
  MaterialExpressionConstant_0 {class=MaterialExpressionConstant}
  MaterialExpressionConstant_1 {class=MaterialExpressionConstant}
  MaterialExpressionConstant_2 {class=MaterialExpressionConstant, properties {r=0.6}}
  MaterialExpressionConstant_3 {class=MaterialExpressionConstant}
  MaterialExpressionConvert_0 {
   class=MaterialExpressionConvert
   inputs [None {from=MaterialExpressionMultiply_6}]
   properties {
    nodeName=Break Float3
    convertInputs [{
      type=Vector3
      expressionInput {
       expression=&self:MaterialExpressionMultiply_6
       inputName=None
       mask=0
       maskA=0
       maskB=0
       maskG=0
       maskR=0
       outputIndex=0
      }
     }]
    convertOutputs [{type=Scalar}, {type=Scalar}, {type=Scalar}]
   }
  }
  MaterialExpressionConvert_1 {
   class=MaterialExpressionConvert
   inputs [None {from=MaterialExpressionMultiply_11}]
   properties {
    nodeName=Break Float3
    convertInputs [{
      type=Vector3
      expressionInput {
       expression=&self:MaterialExpressionMultiply_11
       inputName=None
       mask=0
       maskA=0
       maskB=0
       maskG=0
       maskR=0
       outputIndex=0
      }
     }]
    convertOutputs [{type=Scalar}, {type=Scalar}, {type=Scalar}]
   }
  }
  MaterialExpressionConvert_2 {
   class=MaterialExpressionConvert
   inputs [None {from=MaterialExpressionDivide_3}]
   properties {
    nodeName=Break Float3
    convertInputs [{
      type=Vector3
      expressionInput {
       expression=&self:MaterialExpressionDivide_3
       inputName=None
       mask=0
       maskA=0
       maskB=0
       maskG=0
       maskR=0
       outputIndex=0
      }
     }]
    convertOutputs [{type=Scalar}, {type=Scalar}, {type=Scalar}]
   }
  }
  MaterialExpressionConvert_3 {
   class=MaterialExpressionConvert
   inputs [
    None {from=MaterialExpressionConvert_2}
    None {from=MaterialExpressionConvert_2}
   ]
   properties {
    nodeName=Make Vector2
    convertInputs [
     {
      type=Scalar
      expressionInput {
       expression=&self:MaterialExpressionConvert_2
       inputName=None
       mask=0
       maskA=0
       maskB=0
       maskG=0
       maskR=0
       outputIndex=0
      }
     }
     {
      type=Scalar
      expressionInput {
       expression=&self:MaterialExpressionConvert_2
       inputName=None
       mask=0
       maskA=0
       maskB=0
       maskG=0
       maskR=0
       outputIndex=1
      }
     }
    ]
    convertOutputs [{type=Vector2}]
   }
  }
  MaterialExpressionCurveAtlasRowParameter_1 {
   class=MaterialExpressionCurveAtlasRowParameter
   inputs [CurveTime {from=MaterialExpressionSaturate_3, output=""}]
   properties {
    atlas=/BRCosmetics/Characters/Player/Male/Medium/Bodies/M_MED_FloralMane_Sand/Skins/SuperLevels/Textures/CA_SL_S39
    curve=/BRCosmetics/Characters/Player/Male/Medium/Bodies/M_MED_FloralMane_Sand/Skins/SuperLevels/Textures/C_SL_S39_CausticsChromatic
    defaultValue=1
    parameterName=PrimaryElement_CausticsChromatic
   }
  }
  MaterialExpressionDesaturation_0 {
   class=MaterialExpressionDesaturation
   inputs [
    None {from=MaterialExpressionCurveAtlasRowParameter_1, output=""}
    Fraction {from=MaterialExpressionScalarParameter_6, output=""}
   ]
   properties {}
  }
  MaterialExpressionDesaturation_1 {
   class=MaterialExpressionDesaturation
   inputs [
    None {from=MaterialExpressionVectorParameter_0, output=RGB}
    Fraction {from=MaterialExpressionConstant_2, output=""}
   ]
   properties {}
  }
  MaterialExpressionDivide_3 {
   class=MaterialExpressionDivide
   inputs [
    A {from=MaterialExpressionMultiply_9}
    B {from=MaterialExpressionScalarParameter_3}
   ]
   properties {constB=64}
  }
  MaterialExpressionFunctionInput_0 {
   class=MaterialExpressionFunctionInput
   inputs [Preview {from="", output=""}]
   properties {inputName=MA, inputType=FunctionInput_MaterialAttributes}
  }
  MaterialExpressionFunctionInput_1 {class=MaterialExpressionFunctionInput, inputs [Preview {from="", output=""}], properties {inputName=PreskinnedPosition, sortPriority=1}}
  MaterialExpressionFunctionInput_2 {
   class=MaterialExpressionFunctionInput
   inputs [Preview {from="", output=""}]
   properties {bUsePreviewValueAsDefault=true, inputName=TechArtMask, sortPriority=1}
  }
  MaterialExpressionFunctionInput_3 {
   class=MaterialExpressionFunctionInput
   inputs [Preview {from=MaterialExpressionConstant_1, output=""}]
   properties {bUsePreviewValueAsDefault=true, inputName=InFresnel, inputType=FunctionInput_Scalar, sortPriority=3}
  }
  MaterialExpressionFunctionOutput_1 {class=MaterialExpressionFunctionOutput, inputs [None {from=MaterialExpressionSetMaterialAttributes_0, output=""}]}
  MaterialExpressionGetMaterialAttributes_0 {
   class=MaterialExpressionGetMaterialAttributes
   inputs [None {from=MaterialExpressionFunctionInput_0, output=""}]
   properties {attributeGetTypes [#1, #2]}
  }
  MaterialExpressionLinearInterpolate_0 {
   class=MaterialExpressionLinearInterpolate
   inputs [
    A {from=""}
    B {from=MaterialExpressionComponentMask_0}
    Alpha {from=MaterialExpressionScalarParameter_2}
   ]
   properties {constA=1}
  }
  MaterialExpressionLinearInterpolate_1 {
   class=MaterialExpressionLinearInterpolate
   inputs [
    A {from=MaterialExpressionGetMaterialAttributes_0, output=BaseColor}
    B {from=MaterialExpressionMultiply_3}
    Alpha {from=MaterialExpressionLinearInterpolate_0}
   ]
  }
  MaterialExpressionMaterialFunctionCall_1 {
   class=MaterialExpressionMaterialFunctionCall
   inputs [
    TextureObject {from=MaterialExpressionTextureObject_0}
    TextureSize {from=MaterialExpressionScalarParameter_9}
    WorldPosition {from=MaterialExpressionAdd_9}
    FadingGradient {from=""}
    In {from=MaterialExpressionConstant2Vector_2}
    World Space Normal {from=""}
    ProjectionTransitionContrast {from=MaterialExpressionConstant_3}
   ]
   properties {materialFunction=/BRCosmetics/Characters/Player/Male/Medium/Bodies/M_MED_FloralMane_Sand/Skins/SuperLevels/ParentMaterials/Preskinned_WP_Texture_ScalePan_XY}
  }
  MaterialExpressionMaterialFunctionCall_2 {
   class=MaterialExpressionMaterialFunctionCall
   inputs [
    TextureObject {from=MaterialExpressionTextureObject_0}
    TextureSize {from=MaterialExpressionScalarParameter_23}
    WorldPosition {from=MaterialExpressionAdd_17}
    FadingGradient {from=""}
    In {from=MaterialExpressionConstant2Vector_2}
    World Space Normal {from=""}
    ProjectionTransitionContrast {from=MaterialExpressionConstant_0}
   ]
   properties {materialFunction=&Preskinned_WP_Texture_ScalePan_XY}
  }
  MaterialExpressionMultiply_0 {
   class=MaterialExpressionMultiply
   inputs [
    A {from=MaterialExpressionReroute_1}
    B {from=MaterialExpressionScalarParameter_1}
   ]
  }
  MaterialExpressionMultiply_1 {
   class=MaterialExpressionMultiply
   inputs [
    A {from=MaterialExpressionLinearInterpolate_0}
    B {from=MaterialExpressionAdd_16}
   ]
  }
  MaterialExpressionMultiply_10 {
   class=MaterialExpressionMultiply
   inputs [
    A {from=MaterialExpressionSaturate_1}
    B {from=MaterialExpressionMaterialFunctionCall_1, output=XY Texture}
   ]
  }
  MaterialExpressionMultiply_11 {
   class=MaterialExpressionMultiply
   inputs [
    A {from=MaterialExpressionMultiply_10}
    B {from=MaterialExpressionScalarParameter_10}
   ]
  }
  MaterialExpressionMultiply_12 {
   class=MaterialExpressionMultiply
   inputs [
    A {from=MaterialExpressionNormalize_0}
    B {from=MaterialExpressionScalarParameter_11}
   ]
   properties {constB=12}
  }
  MaterialExpressionMultiply_17 {
   class=MaterialExpressionMultiply
   inputs [
    A {from=MaterialExpressionVectorParameter_0, output=RGB}
    B {from=MaterialExpressionConvert_0}
   ]
  }
  MaterialExpressionMultiply_18 {
   class=MaterialExpressionMultiply
   inputs [
    A {from=MaterialExpressionConvert_0}
    B {from=MaterialExpressionVectorParameter_1, output=RGB}
   ]
  }
  MaterialExpressionMultiply_2 {
   class=MaterialExpressionMultiply
   inputs [
    A {from=MaterialExpressionNamedRerouteUsage_2}
    B {from=MaterialExpressionScalarParameter_14}
   ]
  }
  MaterialExpressionMultiply_20 {
   class=MaterialExpressionMultiply
   inputs [
    A {from=MaterialExpressionNamedRerouteUsage_4}
    B {from=MaterialExpressionScalarParameter_15}
   ]
  }
  MaterialExpressionMultiply_21 {
   class=MaterialExpressionMultiply
   inputs [
    A {from=MaterialExpressionTime_2}
    B {from=MaterialExpressionScalarParameter_18}
   ]
  }
  MaterialExpressionMultiply_22 {
   class=MaterialExpressionMultiply
   inputs [
    A {from=MaterialExpressionDesaturation_1}
    B {from=MaterialExpressionConstantBiasScale_1}
   ]
  }
  MaterialExpressionMultiply_23 {
   class=MaterialExpressionMultiply
   inputs [
    A {from=MaterialExpressionMultiply_22}
    B {from=""}
   ]
   properties {constB=0.1}
  }
  MaterialExpressionMultiply_25 {
   class=MaterialExpressionMultiply
   inputs [
    A {from=MaterialExpressionNormalize_0}
    B {from=MaterialExpressionScalarParameter_24}
   ]
   properties {constB=12}
  }
  MaterialExpressionMultiply_28 {
   class=MaterialExpressionMultiply
   inputs [
    A {from=MaterialExpressionSaturate_0}
    B {from=MaterialExpressionVectorParameter_0, output=RGB}
   ]
  }
  MaterialExpressionMultiply_29 {
   class=MaterialExpressionMultiply
   inputs [
    A {from=MaterialExpressionNamedRerouteDeclaration_16}
    B {from=MaterialExpressionScalarParameter_25}
   ]
   properties {constB=5}
  }
  MaterialExpressionMultiply_3 {
   class=MaterialExpressionMultiply
   inputs [
    A {from=MaterialExpressionGetMaterialAttributes_0, output=BaseColor}
    B {from=MaterialExpressionScalarParameter_12}
   ]
  }
  MaterialExpressionMultiply_30 {
   class=MaterialExpressionMultiply
   inputs [
    A {from=MaterialExpressionDesaturation_0}
    B {from=MaterialExpressionSaturate_0}
   ]
  }
  MaterialExpressionMultiply_36 {
   class=MaterialExpressionMultiply
   inputs [
    A {from=MaterialExpressionDesaturation_0}
    B {from=MaterialExpressionScalarParameter_0}
   ]
   properties {constB=0.2}
  }
  MaterialExpressionMultiply_4 {
   class=MaterialExpressionMultiply
   inputs [
    A {from=MaterialExpressionPower_0}
    B {from=MaterialExpressionScalarParameter_5}
   ]
  }
  MaterialExpressionMultiply_5 {
   class=MaterialExpressionMultiply
   inputs [
    A {from=MaterialExpressionMultiply_30}
    B {from=MaterialExpressionVectorParameter_0, output=RGB}
   ]
  }
  MaterialExpressionMultiply_6 {
   class=MaterialExpressionMultiply
   inputs [
    A {from=MaterialExpressionSaturate_1}
    B {from=MaterialExpressionMaterialFunctionCall_2, output=XY Texture}
   ]
  }
  MaterialExpressionMultiply_7 {
   class=MaterialExpressionMultiply
   inputs [
    A {from=MaterialExpressionPower_1}
    B {from=""}
   ]
   properties {constB=2}
  }
  MaterialExpressionMultiply_8 {
   class=MaterialExpressionMultiply
   inputs [
    A {from=MaterialExpressionTime_1}
    B {from=MaterialExpressionScalarParameter_7}
   ]
   properties {constB=-0.3333}
  }
  MaterialExpressionMultiply_9 {
   class=MaterialExpressionMultiply
   inputs [
    A {from=MaterialExpressionFunctionInput_1}
    B {from=MaterialExpressionAppendVector_0}
   ]
  }
  MaterialExpressionNamedRerouteDeclaration_0 {
   class=MaterialExpressionNamedRerouteDeclaration
   inputs [None {from=MaterialExpressionFunctionInput_3}]
   properties {name=InFresnel, nodeColor=<rgba 1 0 0.0234 1>}
  }
  MaterialExpressionNamedRerouteDeclaration_1 {
   class=MaterialExpressionNamedRerouteDeclaration
   inputs [None {from=MaterialExpressionFunctionInput_1}]
   properties {name=PreskinnedLocalPosition, nodeColor=<rgba 1 0.7266 0 1>}
  }
  MaterialExpressionNamedRerouteDeclaration_16 {
   class=MaterialExpressionNamedRerouteDeclaration
   inputs [None {from=MaterialExpressionTextureSample_4, output=R}]
   properties {name=CausticsDistortionGradient, nodeColor=<rgba 0 0.1797 1 1>}
  }
  MaterialExpressionNamedRerouteDeclaration_9 {
   class=MaterialExpressionNamedRerouteDeclaration
   inputs [None {from=MaterialExpressionMultiply_1}]
   properties {name=Caustics, nodeColor=<rgba 0 0.5312 1 1>}
  }
  MaterialExpressionNamedRerouteUsage_0 {class=MaterialExpressionNamedRerouteUsage}
  MaterialExpressionNamedRerouteUsage_2 {class=MaterialExpressionNamedRerouteUsage}
  MaterialExpressionNamedRerouteUsage_4 {class=MaterialExpressionNamedRerouteUsage}
  MaterialExpressionNoise_0 {
   class=MaterialExpressionNoise
   inputs [
    World Position {from=MaterialExpressionMultiply_20, output=""}
    FilterWidth {from="", output=""}
   ]
   properties {levels=3, outputMin=0}
  }
  MaterialExpressionNormalize_0 {class=MaterialExpressionNormalize, inputs [VectorInput {from=MaterialExpressionSubtract_2, output=""}]}
  MaterialExpressionOneMinus_0 {class=MaterialExpressionOneMinus, inputs [None {from=MaterialExpressionNamedRerouteUsage_0}]}
  MaterialExpressionPanner_3 {
   class=MaterialExpressionPanner
   inputs [
    Coordinate {from=MaterialExpressionConvert_3}
    Time {from=""}
    Speed {from=""}
   ]
   properties {speedX=0.053, speedY=-0.1124}
  }
  MaterialExpressionPower_0 {
   class=MaterialExpressionPower
   inputs [
    Base {from=MaterialExpressionNamedRerouteDeclaration_0}
    Exp {from=MaterialExpressionScalarParameter_4}
   ]
  }
  MaterialExpressionPower_1 {
   class=MaterialExpressionPower
   inputs [
    Base {from=MaterialExpressionOneMinus_0}
    Exp {from=""}
   ]
   properties {constExponent=3.5}
  }
  MaterialExpressionQualitySwitch_0 {
   class=MaterialExpressionQualitySwitch
   inputs [
    Default {from=MaterialExpressionAdd_15, output=""}
    Low {from=MaterialExpressionReroute_0, output=""}
    High {from="", output=""}
    Medium {from="", output=""}
    Epic {from="", output=""}
   ]
   properties {}
  }
  MaterialExpressionQualitySwitch_1 {
   class=MaterialExpressionQualitySwitch
   inputs [
    Default {from=MaterialExpressionAdd_8, output=""}
    Low {from=MaterialExpressionReroute_0, output=""}
    High {from="", output=""}
    Medium {from="", output=""}
    Epic {from="", output=""}
   ]
   properties {}
  }
  MaterialExpressionReroute_0 {class=MaterialExpressionReroute, inputs [None {from=MaterialExpressionMultiply_9}]}
  MaterialExpressionReroute_1 {class=MaterialExpressionReroute, inputs [None {from=MaterialExpressionAdd_20}]}
  MaterialExpressionSaturate_0 {class=MaterialExpressionSaturate, inputs [None {from=MaterialExpressionAdd_4}]}
  MaterialExpressionSaturate_1 {class=MaterialExpressionSaturate, inputs [None {from=MaterialExpressionMultiply_7}]}
  MaterialExpressionSaturate_2 {class=MaterialExpressionSaturate, inputs [None {from=MaterialExpressionAdd_6}]}
  MaterialExpressionSaturate_3 {class=MaterialExpressionSaturate, inputs [None {from=MaterialExpressionAdd_7}]}
  MaterialExpressionScalarParameter_0 {class=MaterialExpressionScalarParameter, properties {defaultValue=0.02, group=Caustics, parameterName=BaseEmissive_Influence}}
  MaterialExpressionScalarParameter_1 {class=MaterialExpressionScalarParameter, properties {defaultValue=1, group=Caustics, parameterName=Caustics_Intensity}}
  MaterialExpressionScalarParameter_10 {class=MaterialExpressionScalarParameter, properties {defaultValue=0.5, group=Caustics, parameterName=Caustics02_Intensity}}
  MaterialExpressionScalarParameter_11 {class=MaterialExpressionScalarParameter, properties {defaultValue=15, group=Caustics, parameterName=Caustics02_ParallaxDepth}}
  MaterialExpressionScalarParameter_12 {class=MaterialExpressionScalarParameter, properties {defaultValue=1, parameterName=BaseColorMultiplier}}
  MaterialExpressionScalarParameter_14 {class=MaterialExpressionScalarParameter, properties {group=Caustics, parameterName=SaltLattice_UseCaustics}}
  MaterialExpressionScalarParameter_15 {class=MaterialExpressionScalarParameter, properties {defaultValue=0.02, parameterName=Wripples_Scale}}
  MaterialExpressionScalarParameter_18 {class=MaterialExpressionScalarParameter, properties {defaultValue=0.275, parameterName=Wripples_Speed}}
  MaterialExpressionScalarParameter_2 {class=MaterialExpressionScalarParameter, properties {group=Caustics, parameterName=Caustics_UseTechArtMask}}
  MaterialExpressionScalarParameter_23 {class=MaterialExpressionScalarParameter, properties {defaultValue=50, group=Caustics, parameterName=Caustics_Size}}
  MaterialExpressionScalarParameter_24 {class=MaterialExpressionScalarParameter, properties {defaultValue=15, group=Caustics, parameterName=Caustics_ParallaxDepth}}
  MaterialExpressionScalarParameter_25 {class=MaterialExpressionScalarParameter, properties {defaultValue=7, group=Caustics, parameterName=Caustics_Distortion}}
  MaterialExpressionScalarParameter_3 {class=MaterialExpressionScalarParameter, properties {defaultValue=64, group=Caustics, parameterName=Caustics_DistortionScale}}
  MaterialExpressionScalarParameter_4 {class=MaterialExpressionScalarParameter, properties {defaultValue=2, group=Caustics, parameterName=Caustics_FresnelPower}}
  MaterialExpressionScalarParameter_5 {class=MaterialExpressionScalarParameter, properties {defaultValue=1, group=Caustics, parameterName=Caustics_FresnelIntensity}}
  MaterialExpressionScalarParameter_6 {class=MaterialExpressionScalarParameter, properties {defaultValue=0.5, group=Caustics, parameterName=Caustics_ChromaticDesaturation}}
  MaterialExpressionScalarParameter_7 {class=MaterialExpressionScalarParameter, properties {defaultValue=-0.3333, group=Caustics, parameterName=Caustics_DistortionSpeed}}
  MaterialExpressionScalarParameter_8 {class=MaterialExpressionScalarParameter, properties {defaultValue=1, group=Caustics, parameterName=Caustics_VerticalStretch}}
  MaterialExpressionScalarParameter_9 {class=MaterialExpressionScalarParameter, properties {defaultValue=20, group=Caustics, parameterName=Caustics02_Size}}
  MaterialExpressionSetMaterialAttributes_0 {
   class=MaterialExpressionSetMaterialAttributes
   inputs [
    MaterialAttributes {from=MaterialExpressionGetMaterialAttributes_0, output=MaterialAttributes}
    EmissiveColor {from=MaterialExpressionAdd_0, output=""}
    BaseColor {from=MaterialExpressionLinearInterpolate_1, output=""}
   ]
   properties {attributeSetTypes [#1, #2]}
  }
  MaterialExpressionShadingPathSwitch_0 {
   class=MaterialExpressionShadingPathSwitch
   inputs [
    Default {from=MaterialExpressionQualitySwitch_0, output=""}
    Deferred {from=MaterialExpressionQualitySwitch_0, output=""}
    Forward {from=MaterialExpressionReroute_0, output=""}
    Mobile {from=MaterialExpressionReroute_0, output=""}
   ]
   properties {}
  }
  MaterialExpressionShadingPathSwitch_1 {
   class=MaterialExpressionShadingPathSwitch
   inputs [
    Default {from=MaterialExpressionQualitySwitch_1, output=""}
    Deferred {from=MaterialExpressionQualitySwitch_1, output=""}
    Forward {from="", output=""}
    Mobile {from="", output=""}
   ]
   properties {}
  }
  MaterialExpressionSine_0 {class=MaterialExpressionSine, inputs [None {from=MaterialExpressionAdd_3}]}
  MaterialExpressionSine_1 {class=MaterialExpressionSine, inputs [None {from=MaterialExpressionAdd_13}]}
  MaterialExpressionSubtract_2 {
   class=MaterialExpressionSubtract
   inputs [
    A {from=MaterialExpressionMultiply_9}
    B {from=MaterialExpressionTransformPosition_0}
   ]
  }
  MaterialExpressionTextureObject_0 {
   class=MaterialExpressionTextureObject
   inputs []
   properties {
    samplerType=SAMPLERTYPE_Masks
    texture=/BRCosmetics/Characters/Player/Male/Medium/Bodies/M_MED_FloralMane_Sand/Skins/SuperLevels/Textures/T_SL_S39_T1_TA
   }
  }
  MaterialExpressionTextureSample_4 {
   class=MaterialExpressionTextureSample
   inputs [
    UVs {from=MaterialExpressionAdd_2}
    Tex {from=""}
    Apply View MipBias {from=""}
   ]
   properties {
    samplerSource=SSM_Wrap_WorldGroupSettings
    samplerType=SAMPLERTYPE_Masks
    texture=&T_SL_S39_T1_TA
   }
  }
  MaterialExpressionTime_1 {class=MaterialExpressionTime}
  MaterialExpressionTime_2 {class=MaterialExpressionTime}
  MaterialExpressionTransformPosition_0 {
   class=MaterialExpressionTransformPosition
   inputs [None {from=MaterialExpressionCameraPositionWS_0, output=""}]
   properties {transformSourceType=TRANSFORMPOSSOURCE_World}
  }
  MaterialExpressionVectorParameter_0 {class=MaterialExpressionVectorParameter, properties {group=Caustics, parameterName=Caustics_Tint, defaultValue=<rgba 0 0.2601 0.5771 1>}}
  MaterialExpressionVectorParameter_1 {
   class=MaterialExpressionVectorParameter
   properties {group=Caustics, parameterName=Caustics_Tint_Chromatic, defaultValue=<rgba 0 0.5348 0.5771 1>}
  }
 }}
