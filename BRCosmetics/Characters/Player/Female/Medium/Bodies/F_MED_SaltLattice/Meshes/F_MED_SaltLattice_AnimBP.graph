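# Animation Blueprint for the female SaltLattice cosmetic body, matching its jaw to the attached head and adding secondary motion to fins.
# It copies the parent mesh pose through a Control Rig, then replaces chin and jaw bone transforms before spring-driven AnimDynamics moves arm and leg fins.
# Each update reads those facial transforms from the customization-provided head mesh when it has a jaw bone, and also updates a gravity override variable.
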AnimBlueprint /BRCosmetics/Characters/Player/Female/Medium/Bodies/F_MED_SaltLattice/Meshes/F_MED_SaltLattice_AnimBP
target_skeleton=&/F_MED_SaltLattice_Skeleton
graphs [
 AnimGraph {nodes [
   n1 = Anim Dynamics - Bone: dyn_Armfin_l!(ComponentPose=Local To Component(LocalPose=Use cached pose 'HeadCH'()), Alpha=0.8, GravityOverride:Vector="0,0,-980") -> (Pose) {
    angularDampingOverride=1
    angularSpringConstant=9000
    bAngularSpring=true
    bOverrideAngularDamping=true
    bUseGravityOverride=true
    gravityScale=0.35
    simulationSpace=RootRelative
    boundBone {boneName=dyn_Armfin_l}
    componentAppliedLinearAccClamp=<5000 5000 5000>
    componentLinearAccScale=<2 2 2>
    componentLinearVelScale=<2 2 2>
    gravityOverride {z=-980}
    physicsBodyDefinitions [{
      boundBone {boneName=dyn_Armfin_l}
      boxExtents=<20 40 20>
      constraintSetup {angularLimitsMax=<0 0 5>, angularLimitsMin=<0 -15 -20>, angularTarget=<1 0 0>}
      localJointOffset=<6 0 0>
     }]
   }
   n2 = Anim Dynamics - Bone: dyn_Armfin_r!(ComponentPose=@n1, Alpha=0.8, GravityOverride:Vector="0,0,-980") -> (Pose) {
    angularDampingOverride=1
    angularSpringConstant=9000
    bAngularSpring=true
    bOverrideAngularDamping=true
    bUseGravityOverride=true
    gravityScale=0.35
    simulationSpace=RootRelative
    boundBone {boneName=dyn_Armfin_r}
    componentAppliedLinearAccClamp=<5000 5000 5000>
    componentLinearAccScale=<2 2 2>
    componentLinearVelScale=<2 2 2>
    gravityOverride {z=-980}
    physicsBodyDefinitions [{
      boundBone {boneName=dyn_Armfin_r}
      boxExtents=<20 40 20>
      constraintSetup {angularLimitsMax=<0 0 5>, angularLimitsMin=<0 -15 -20>, angularTarget=<1 0 0>}
      localJointOffset=<6 0 0>
     }]
   }
   n3 = Anim Dynamics - Bone: dyn_Legfin_l!(ComponentPose=@n2, Alpha=1, GravityOverride:Vector="0,0,-980") -> (Pose) {
    angularDampingOverride=1
    angularSpringConstant=9000
    bAngularSpring=true
    bOverrideAngularDamping=true
    bUseGravityOverride=true
    gravityScale=0.35
    simulationSpace=RootRelative
    boundBone {boneName=dyn_Legfin_l}
    componentAppliedLinearAccClamp=<5000 5000 5000>
    componentLinearAccScale=<2 2 2>
    componentLinearVelScale=<2 2 2>
    gravityOverride {z=-980}
    physicsBodyDefinitions [{
      boundBone {boneName=dyn_Legfin_l}
      boxExtents=<20 40 20>
      constraintSetup {angularLimitsMax=<0 30 15>, angularLimitsMin=<0 -30 -15>, angularTarget=<1 0 0>}
      localJointOffset=<6 0 0>
     }]
   }
   n4 = Anim Dynamics - Bone: dyn_Legfin_r!(ComponentPose=@n3, Alpha=1, GravityOverride:Vector="0,0,-980") -> (Pose) {
    angularDampingOverride=1
    angularSpringConstant=9000
    bAngularSpring=true
    bOverrideAngularDamping=true
    bUseGravityOverride=true
    gravityScale=0.35
    simulationSpace=RootRelative
    boundBone {boneName=dyn_Legfin_r}
    componentAppliedLinearAccClamp=<5000 5000 5000>
    componentLinearAccScale=<2 2 2>
    componentLinearVelScale=<2 2 2>
    gravityOverride {z=-980}
    physicsBodyDefinitions [{
      boundBone {boneName=dyn_Legfin_r}
      boxExtents=<20 40 20>
      constraintSetup {angularLimitsMax=<0 30 15>, angularLimitsMin=<0 -30 -15>, angularTarget=<1 0 0>}
      localJointOffset=<6 0 0>
     }]
   }
   Output Pose(Result=Component To Local(ComponentPose=@n4)) {layerGroup=None}
   n5 = Control Rig(Source=Copy Pose From Mesh() {bUseAttachedParent=true}) {controlRigAssetReference {blueprintRigClass=&/F_MED_SaltLattice_CtrlRig.F_MED_SaltLattice_CtrlRig_C}}
   Save cached pose 'InputCH'(Pose=@n5) {cacheName=InputCH}
   n6 = "Transform (Modify) Bone - Bone: FACIAL_C_UnderChin"!(
    ComponentPose=Local To Component(LocalPose=Use cached pose 'InputCH'())
    bAlphaBoolEnabled:bool=True
    Translation=Get FACIAL_C_UnderChin_Translation() -> (FACIAL_C_UnderChin_Translation:Vector)
    Rotation=Get FACIAL_C_UnderChin_Rotation() -> (FACIAL_C_UnderChin_Rotation:Rotator)
    Scale="1,1,1"
   ) -> (Pose) {
    alphaInputType=Bool
    rotationMode=BMM_Replace
    rotationSpace=BCS_ParentBoneSpace
    translationMode=BMM_Replace
    translationSpace=BCS_ParentBoneSpace
    alphaBoolBlend {blendInTime=0.25, blendOption=Linear, blendOutTime=0.25}
    boneToModify {boneName=FACIAL_C_UnderChin}
   }
   n7 = "Transform (Modify) Bone - Bone: FACIAL_C_Chin"!(
    ComponentPose=@n6
    bAlphaBoolEnabled:bool=True
    Translation=Get FACIAL_C_Chin_Translation() -> (FACIAL_C_Chin_Translation:Vector)
    Rotation=Get FACIAL_C_Chin_Rotation() -> (FACIAL_C_Chin_Rotation:Rotator)
    Scale="1,1,1"
   ) -> (Pose) {
    alphaInputType=Bool
    rotationMode=BMM_Replace
    rotationSpace=BCS_ParentBoneSpace
    translationMode=BMM_Replace
    translationSpace=BCS_ParentBoneSpace
    alphaBoolBlend {blendOption=Linear}
    boneToModify {boneName=FACIAL_C_Chin}
   }
   n8 = "Transform (Modify) Bone - Bone: FACIAL_L_UnderChin"!(
    ComponentPose=@n7
    bAlphaBoolEnabled:bool=True
    Translation=Get FACIAL_L_UnderChin_Translation() -> (FACIAL_L_UnderChin_Translation:Vector)
    Rotation=Get FACIAL_L_UnderChin_Rotation() -> (FACIAL_L_UnderChin_Rotation:Rotator)
    Scale="1,1,1"
   ) -> (Pose) {
    alphaInputType=Bool
    rotationMode=BMM_Replace
    rotationSpace=BCS_ParentBoneSpace
    translationMode=BMM_Replace
    translationSpace=BCS_ParentBoneSpace
    alphaBoolBlend {blendOption=Linear}
    boneToModify {boneName=FACIAL_L_UnderChin}
   }
   n9 = "Transform (Modify) Bone - Bone: FACIAL_R_UnderChin"!(
    ComponentPose=@n8
    bAlphaBoolEnabled:bool=True
    Translation=Get FACIAL_R_UnderChin_Translation() -> (FACIAL_R_UnderChin_Translation:Vector)
    Rotation=Get FACIAL_R_UnderChin_Rotation() -> (FACIAL_R_UnderChin_Rotation:Rotator)
    Scale="1,1,1"
   ) -> (Pose) {
    alphaInputType=Bool
    rotationMode=BMM_Replace
    rotationSpace=BCS_ParentBoneSpace
    translationMode=BMM_Replace
    translationSpace=BCS_ParentBoneSpace
    alphaBoolBlend {blendOption=Linear}
    boneToModify {boneName=FACIAL_R_UnderChin}
   }
   n10 = "Transform (Modify) Bone - Bone: FACIAL_L_Jawline"!(
    ComponentPose=@n9
    bAlphaBoolEnabled:bool=True
    Translation=Get FACIAL_L_Jawline_Translation() -> (FACIAL_L_Jawline_Translation:Vector)
    Rotation=Get FACIAL_L_Jawline_Rotation() -> (FACIAL_L_Jawline_Rotation:Rotator)
    Scale="1,1,1"
   ) -> (Pose) {
    alphaInputType=Bool
    rotationMode=BMM_Replace
    rotationSpace=BCS_ParentBoneSpace
    translationMode=BMM_Replace
    translationSpace=BCS_ParentBoneSpace
    alphaBoolBlend {blendOption=Linear}
    boneToModify {boneName=FACIAL_L_Jawline}
   }
   n11 = "Transform (Modify) Bone - Bone: FACIAL_R_Jawline"!(
    ComponentPose=@n10
    bAlphaBoolEnabled:bool=True
    Translation=Get FACIAL_R_Jawline_Translation() -> (FACIAL_R_Jawline_Translation:Vector)
    Rotation=Get FACIAL_R_Jawline_Rotation() -> (FACIAL_R_Jawline_Rotation:Rotator)
    Scale="1,1,1"
   ) -> (Pose) {
    alphaInputType=Bool
    rotationMode=BMM_Replace
    rotationSpace=BCS_ParentBoneSpace
    translationMode=BMM_Replace
    translationSpace=BCS_ParentBoneSpace
    alphaBoolBlend {blendOption=Linear}
    boneToModify {boneName=FACIAL_R_Jawline}
   }
   n12 = "Transform (Modify) Bone - Bone: FACIAL_C_Jaw"!(
    ComponentPose=@n11
    bAlphaBoolEnabled:bool=True
    Translation=Get FACIAL_C_Jaw_Tranform() -> (FACIAL_C_Jaw_Tranform:Vector)
    Rotation=Get FACIAL_C_Jaw_Rotation() -> (FACIAL_C_Jaw_Rotation:Rotator)
    Scale="1,1,1"
   ) -> (Pose) {
    alphaInputType=Bool
    rotationMode=BMM_Replace
    rotationSpace=BCS_ParentBoneSpace
    translationMode=BMM_Replace
    translationSpace=BCS_ParentBoneSpace
    alphaBoolBlend {blendOption=Linear}
    boneToModify {boneName=FACIAL_C_Jaw}
   }
   n13 = "Transform (Modify) Bone - Bone: FACIAL_C_Jawline"!(
    ComponentPose=@n12
    bAlphaBoolEnabled:bool=True
    Translation=Get FACIAL_C_Jawline_Translation() -> (FACIAL_C_Jawline_Translation:Vector)
    Rotation=Get FACIAL_C_Jawline_Rotation() -> (FACIAL_C_Jawline_Rotation:Rotator)
    Scale="1,1,1"
   ) -> (Pose) {
    alphaInputType=Bool
    rotationMode=BMM_Replace
    rotationSpace=BCS_ParentBoneSpace
    translationMode=BMM_Replace
    translationSpace=BCS_ParentBoneSpace
    alphaBoolBlend {blendOption=Linear}
    boneToModify {boneName=FACIAL_C_Jawline}
   }
   Save cached pose 'HeadCH'(Pose=Component To Local(ComponentPose=@n13)) {cacheName=HeadCH}
  ]}
 EventGraph {nodes [
   Event BlueprintUpdateAnimation()
   > n1 = Sequence()
   > n2 = "Gravity Override State Function (Parameter Asset) (Nativized)"(GameTimeInSeconds=Get Game Time in Seconds() -> (Game Time in Seconds:float))
   > Set Gravity Override(Gravity Override=@n2)
   n3 = Get HeadMesh() -> (HeadMesh:Skeletal Mesh Component Budgeted Object Reference)
   Is Valid(exec=@n1.then_1, InputObject=@n3)
   n4 = GetSocketBoneName!(self:Skinned Mesh Component Object Reference=@n3, InSocketName:Name=FACIAL_C_Jaw) -> (ReturnValue:Name=None)
   > n5 = Set isUsingFacialBones(isUsingFacialBones:bool="Not Equal Exactly (String)"(A="To String (Name)"(InName=@n4), B=None)) -> (Output_Get:bool)
   > Branch(Condition=@n5)
   n6 = Get HeadMesh()
   > K2Node_Composite_0 = JawLineSetup!(
    execute
    self:Scene Component Object Reference=@n6
    self2:Scene Component Object Reference=@n6
    self3:Scene Component Object Reference=@n6
    self4:Scene Component Object Reference=@n6
    self5:Scene Component Object Reference=@n6
    self6:Scene Component Object Reference=@n6
    self7:Scene Component Object Reference=@n6
    self8:Scene Component Object Reference=@n6
   ) {is=title K2Node_Composite}
   n7 = Event OnCharacterCustomizationLoaded()
   > Set HeadMesh(HeadMesh:Skeletal Mesh Component Budgeted Object Reference=Get Array Item with EFort Custom Part Type(Array=@n7.CustomMeshParts, PartType=Head)) -> (Output_Get:Skeletal Mesh Component Budgeted Object Reference)
  ]}
 JawLineSetup {
  path=EventGraph.K2Node_Composite_0.JawLineSetup
  nodes [
   n1 = Inputs!() -> (
    execute
    self:Scene Component Object Reference
    self2:Scene Component Object Reference
    self3:Scene Component Object Reference
    self4:Scene Component Object Reference
    self5:Scene Component Object Reference
    self6:Scene Component Object Reference
    self7:Scene Component Object Reference
    self8:Scene Component Object Reference
   )
   n2 = BreakTransform(InTransform=GetSocketTransform(self=@n1.self6, InSocketName=FACIAL_C_UnderChin, TransformSpace=RTS_ParentBoneSpace))
   > Set FACIAL_C_UnderChin_Translation(FACIAL_C_UnderChin_Translation:Vector=@n2) -> (Output_Get:Vector)
   > Set FACIAL_C_UnderChin_Rotation(FACIAL_C_UnderChin_Rotation:Rotator=@n2.Rotation) -> (Output_Get:Rotator)
   n3 = BreakTransform(InTransform=GetSocketTransform(self=@n1, InSocketName=FACIAL_C_Chin, TransformSpace=RTS_ParentBoneSpace))
   > Set FACIAL_C_Chin_Translation(FACIAL_C_Chin_Translation:Vector=@n3) -> (Output_Get:Vector)
   > Set FACIAL_C_Chin_Rotation(FACIAL_C_Chin_Rotation:Rotator=@n3.Rotation) -> (Output_Get:Rotator)
   n4 = BreakTransform(InTransform=GetSocketTransform(self=@n1.self3, InSocketName=FACIAL_L_UnderChin, TransformSpace=RTS_ParentBoneSpace))
   > Set FACIAL_L_UnderChin_Translation(FACIAL_L_UnderChin_Translation:Vector=@n4) -> (Output_Get:Vector)
   > Set FACIAL_L_UnderChin_Rotation(FACIAL_L_UnderChin_Rotation:Rotator=@n4.Rotation) -> (Output_Get:Rotator)
   n5 = BreakTransform(InTransform=GetSocketTransform(self=@n1.self7, InSocketName=FACIAL_R_UnderChin, TransformSpace=RTS_ParentBoneSpace))
   > Set FACIAL_R_UnderChin_Translation(FACIAL_R_UnderChin_Translation:Vector=@n5) -> (Output_Get:Vector)
   > Set FACIAL_R_UnderChin_Rotation(FACIAL_R_UnderChin_Rotation:Rotator=@n5.Rotation) -> (Output_Get:Rotator)
   n6 = BreakTransform(InTransform=GetSocketTransform(self=@n1.self8, InSocketName=FACIAL_C_Jaw, TransformSpace=RTS_ParentBoneSpace))
   > Set FACIAL_C_Jaw_Tranform(FACIAL_C_Jaw_Tranform:Vector=@n6) -> (Output_Get:Vector)
   > Set FACIAL_C_Jaw_Rotation(FACIAL_C_Jaw_Rotation:Rotator=@n6.Rotation) -> (Output_Get:Rotator)
   n7 = BreakTransform(InTransform=GetSocketTransform(self=@n1.self2, InSocketName=FACIAL_C_Jawline, TransformSpace=RTS_ParentBoneSpace))
   > Set FACIAL_C_Jawline_Translation(FACIAL_C_Jawline_Translation:Vector=@n7) -> (Output_Get:Vector)
   > Set FACIAL_C_Jawline_Rotation(FACIAL_C_Jawline_Rotation:Rotator=@n7.Rotation) -> (Output_Get:Rotator)
   n8 = BreakTransform(InTransform=GetSocketTransform(self=@n1.self5, InSocketName=FACIAL_L_Jawline, TransformSpace=RTS_ParentBoneSpace))
   > Set FACIAL_L_Jawline_Translation(FACIAL_L_Jawline_Translation:Vector=@n8) -> (Output_Get:Vector)
   > Set FACIAL_L_Jawline_Rotation(FACIAL_L_Jawline_Rotation:Rotator=@n8.Rotation) -> (Output_Get:Rotator)
   n9 = BreakTransform(InTransform=GetSocketTransform(self=@n1.self4, InSocketName=FACIAL_R_Jawline, TransformSpace=RTS_ParentBoneSpace))
   > Set FACIAL_R_Jawline_Translation(FACIAL_R_Jawline_Translation:Vector=@n9) -> (Output_Get:Vector)
   > Set FACIAL_R_Jawline_Rotation(FACIAL_R_Jawline_Rotation:Rotator=@n9.Rotation) -> (Output_Get:Rotator)
   Outputs()
  ]
 }
]
variables [
 FACIAL_C_Chin_Rotation {type=Rotator}
 FACIAL_C_Chin_Translation {type=Vector}
 FACIAL_C_Jaw_Rotation {type=Rotator}
 FACIAL_C_Jaw_Tranform {type=Vector}
 FACIAL_C_Jawline_Rotation {type=Rotator}
 FACIAL_C_Jawline_Translation {type=Vector}
 FACIAL_C_UnderChin_Rotation {type=Rotator}
 FACIAL_C_UnderChin_Translation {type=Vector}
 FACIAL_L_Jawline_Rotation {type=Rotator}
 FACIAL_L_Jawline_Translation {type=Vector}
 FACIAL_L_UnderChin_Rotation {type=Rotator}
 FACIAL_L_UnderChin_Translation {type=Vector}
 FACIAL_R_Jawline_Rotation {type=Rotator}
 FACIAL_R_Jawline_Translation {type=Vector}
 FACIAL_R_UnderChin_Rotation {type=Rotator}
 FACIAL_R_UnderChin_Translation {type=Vector}
 Game Time in Seconds {type=number}
 Gravity Override {type=Vector}
 HeadMesh {type=SkeletalMeshComponentBudgeted}
 isUsingFacialBones {type=boolean}
]
