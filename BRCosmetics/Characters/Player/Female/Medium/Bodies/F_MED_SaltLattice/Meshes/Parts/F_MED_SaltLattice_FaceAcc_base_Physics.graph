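# Physics asset for the SaltLattice medium female face accessory mesh, setting up three dynamic hair chains (right, left, center) against upper-body capsules.
# Each five-bone chain hangs from a tiny offset capsule that joins neck_02 through a free-rotation constraint.
# Swing limits widen toward the tips, while twist is zero-limit or locked.
# Neck, spine, clavicle and upper-arm capsules form the body volume, with only the right arm constrained to its clavicle.
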PhysicsAsset /BRCosmetics/Characters/Player/Female/Medium/Bodies/F_MED_SaltLattice/Meshes/Parts/F_MED_SaltLattice_FaceAcc_base_Physics
bodies {
 clavicle_l [{
   shapeType=Capsule
   length=7.4228
   radius=5.7886
   shapeName=clavicle_l_capsule
   center=<6.3703 -2.0871 -0.6964>
   rotation=<rot 90 -179.9999 0>
  }]
 clavicle_r [{
   shapeType=Capsule
   length=6.7873
   radius=5.9241
   shapeName=clavicle_r_capsule
   center=<-6.1945 1.3506 1.6487>
   rotation=<rot 85 -0.0002 -0.0004>
  }]
 dyn_hair_01_cr [{shapeType=Capsule, length=16, radius=2, shapeName=dyn_hair_01_rt_capsule, center=<15.2418 0 1.2769>, rotation=<rot 70.0002 -180 -180>}]
 dyn_hair_01_lf [{shapeType=Capsule, length=16, radius=2, shapeName=dyn_hair_01_rt_capsule, center=<15.5058 0 1.3235>, rotation=<rot 80 -180 -180>}]
 dyn_hair_01_rt [{shapeType=Capsule, length=14, radius=4, shapeName=dyn_hair_01_rt_capsule, center=<9.8928 0 2.9939>, rotation=<rot 90 0 0>}]
 dyn_hair_02_cr [{shapeType=Capsule, length=5.8123, radius=1.941, shapeName=dyn_hair_02_rt_capsule}]
 dyn_hair_02_lf [{
   shapeType=Capsule
   length=6.0227
   radius=2.8541
   shapeName=dyn_hair_02_rt_capsule
   center=<5.0095 1.8848 -0.3161>
   rotation=<rot 70.0002 -180 -180>
  }]
 dyn_hair_02_rt [{
   shapeType=Capsule
   length=4.0629
   radius=3.1429
   shapeName=dyn_hair_02_rt_capsule
   center=<5.1981 -2.2662 -0.1569>
   rotation=<rot 80.0001 -180 -180>
  }]
 dyn_hair_03_cr [{shapeType=Capsule, length=5.3259, radius=2.6741, shapeName=dyn_hair_03_cr_capsule, center=<4.8355 0 0.825>, rotation=<rot 79.9999 0 0>}]
 dyn_hair_03_lf [{
   shapeType=Capsule
   length=6.0702
   radius=1.9298
   shapeName=dyn_hair_03_lf_capsule
   center=<4.236 4.8028 0.1369>
   rotation=<rot 85.0001 -180 -180>
  }]
 dyn_hair_03_rt [
  {
   shapeType=Capsule
   length=5.1795
   radius=3.0311
   shapeName=dyn_hair_03_rt_capsule
   center=<3.5457 -5.3943 1.7451>
   rotation=<rot 80.0006 180 180>
  }
  {
   shapeType=Capsule
   length=5.1795
   radius=3.0311
   shapeName=dyn_hair_03_rt_capsule
   center=<3.5457 -1.3356 1.7451>
   rotation=<rot 80.0006 180 180>
  }
 ]
 dyn_hair_04_cr [{shapeType=Capsule, length=8.0522, radius=2.8211, shapeName=dyn_hair_04_cr_capsule, center=<7.3466 0 0>, rotation=<rot 90 0 0>}]
 dyn_hair_04_lf [{shapeType=Capsule, length=8.1395, radius=2.7337, shapeName=dyn_hair_04_lf_capsule, center=<7.3466 2.846 0>, rotation=<rot 90 0 0>}]
 dyn_hair_04_rt [
  {shapeType=Capsule, length=9.2726, radius=1.6006, shapeName=dyn_hair_04_rt_capsule, center=<7.3466 -3.4272 0>, rotation=<rot 90 0 0>}
  {shapeType=Capsule, length=9.2726, radius=1.6006, shapeName=dyn_hair_04_rt_capsule, center=<7.3466 -0.1162 0>, rotation=<rot 90 0 0>}
 ]
 dyn_hair_05_cr [{shapeType=Capsule, length=5, radius=2, shapeName=dyn_hair_05_cr_capsule, center=<5 0 0>, rotation=<rot 90 0 0>}]
 dyn_hair_05_lf [{shapeType=Capsule, length=5, radius=2, shapeName=dyn_hair_05_lf_capsule, center=<5 2.4039 0>, rotation=<rot 90 0 0>}]
 dyn_hair_05_rt [{shapeType=Capsule, length=5.5, radius=1.5, shapeName=dyn_hair_05_rt_capsule, center=<5 0 -1.1027>, rotation=<rot 90 0 0>}]
 neck_02 [{shapeType=Capsule, length=38.4237, radius=12.3308, shapeName=neck_02_capsule, center=<-1.5348 13.5887 0>, rotation=<rot 0 -90 0>}]
 offset_hair_cr [{shapeType=Capsule, length=0.505, radius=0.505, shapeName=offset_hair_cr_capsule, rotation=<rot 0 -90 0>}]
 offset_hair_lf [{shapeType=Capsule, length=0.505, radius=0.505, shapeName=offset_hair_lf_capsule, rotation=<rot 0 -90 0>}]
 offset_hair_rt [{shapeType=Capsule, length=0.505, radius=0.505, shapeName=offset_hair_rt_capsule, rotation=<rot 0 -90 0>}]
 spine_04 [
  {shapeType=Capsule, length=17.1975, radius=7.8025, shapeName=spine_05_capsule, center=<-9.7587 2.9014 0>, rotation=<rot 0 -90 0>}
  {shapeType=Capsule, length=19.0042, radius=7.207, shapeName=spine_04_capsule, center=<0.8439 -0.4724 0>}
 ]
 spine_05 [{shapeType=Capsule, length=17.0286, radius=8.4421, shapeName=spine_05_capsule, center=<2.2882 -0.6259 0>, rotation=<rot 0 -90 0>}]
 upperarm_l [{shapeType=Capsule, length=18.5, radius=5.505, shapeName=upperarm_l_capsule, center=<12.9011 0 0>, rotation=<rot 0 -90 -90>}]
 upperarm_r [{shapeType=Capsule, length=18.5, radius=5.505, shapeName=upperarm_l_capsule, center=<-13.9807 0 0>, rotation=<rot 0 -90 -90>}]
}
constraints [
 {bone1Name=dyn_hair_01_rt, bone2Name=offset_hair_rt, swing1LimitDegrees=0, swing2LimitDegrees=5, twistLimitDegrees=0}
 {bone1Name=dyn_hair_02_rt, bone2Name=dyn_hair_01_rt, swing1LimitDegrees=0, swing2LimitDegrees=20, twistLimitDegrees=0}
 {bone1Name=dyn_hair_03_rt, bone2Name=dyn_hair_02_rt, swing1LimitDegrees=5, swing2LimitDegrees=20, twistLimitDegrees=0}
 {bone1Name=dyn_hair_04_rt, bone2Name=dyn_hair_03_rt, swing1LimitDegrees=5, swing2LimitDegrees=25, twistLimitDegrees=0}
 {bone1Name=dyn_hair_01_lf, bone2Name=offset_hair_lf, swing1LimitDegrees=0, swing2LimitDegrees=5, twistLimitDegrees=0}
 {bone1Name=dyn_hair_02_lf, bone2Name=dyn_hair_01_lf, swing1LimitDegrees=0, swing2LimitDegrees=20, twistLimitDegrees=0}
 {bone1Name=dyn_hair_03_lf, bone2Name=dyn_hair_02_lf, swing1LimitDegrees=5, swing2LimitDegrees=20, twistLimitDegrees=0}
 {bone1Name=dyn_hair_04_lf, bone2Name=dyn_hair_03_lf, swing1LimitDegrees=10, swing2LimitDegrees=25, twistLimitDegrees=0}
 {bone1Name=dyn_hair_01_cr, bone2Name=offset_hair_cr, swing1LimitDegrees=0, swing2LimitDegrees=5, twistLimitDegrees=0}
 {bone1Name=dyn_hair_02_cr, bone2Name=dyn_hair_01_cr, swing1LimitDegrees=0, swing2LimitDegrees=20, twistLimitDegrees=0}
 {bone1Name=dyn_hair_03_cr, bone2Name=dyn_hair_02_cr, swing1LimitDegrees=5, swing2LimitDegrees=20, twistLimitDegrees=0}
 {bone1Name=dyn_hair_04_cr, bone2Name=dyn_hair_03_cr, swing1LimitDegrees=10, swing2LimitDegrees=25, twistLimitDegrees=0}
 {bone1Name=dyn_hair_05_lf, bone2Name=dyn_hair_04_lf, swing1LimitDegrees=10, swing2LimitDegrees=30, twistLimitDegrees=45, twistMotion=Locked}
 {bone1Name=dyn_hair_05_rt, bone2Name=dyn_hair_04_rt, swing1LimitDegrees=10, swing2LimitDegrees=30, twistLimitDegrees=45, twistMotion=Locked}
 {bone1Name=dyn_hair_05_cr, bone2Name=dyn_hair_04_cr, swing1LimitDegrees=10, swing2LimitDegrees=30, twistLimitDegrees=45, twistMotion=Locked}
 {
  bone1Name=offset_hair_lf
  bone2Name=neck_02
  swing1LimitDegrees=45
  swing1Motion=Free
  swing2LimitDegrees=45
  swing2Motion=Free
  twistLimitDegrees=45
  twistMotion=Free
 }
 {
  bone1Name=offset_hair_cr
  bone2Name=neck_02
  swing1LimitDegrees=45
  swing1Motion=Free
  swing2LimitDegrees=45
  swing2Motion=Free
  twistLimitDegrees=45
  twistMotion=Free
 }
 {
  bone1Name=offset_hair_rt
  bone2Name=neck_02
  swing1LimitDegrees=45
  swing1Motion=Free
  swing2LimitDegrees=45
  swing2Motion=Free
  twistLimitDegrees=45
  twistMotion=Free
 }
 {
  bone1Name=upperarm_r
  bone2Name=clavicle_r
  swing1LimitDegrees=45
  swing1Motion=Free
  swing2LimitDegrees=45
  swing2Motion=Free
  twistLimitDegrees=45
  twistMotion=Free
 }
]
