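# Post-sim Control Rig for the SaltLattice female character's face accessory, adding procedural motion to hair and crown bones.
# A sequence first runs ChainHarmonicsPerItem on seven hair and curl chains, using only waves, with pendulum and reach disabled.
# It then rotates each side's crown bone, weighted by the remapped pitch of that side's aim_hair bone relative to its base.
# Three other nodes are unconnected, and the variables are unused.
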ControlRigBlueprint /BRCosmetics/Characters/Player/Female/Medium/Bodies/F_MED_SaltLattice/Meshes/Parts/F_MED_SaltLattice_FaceAcc_postSim_CtrlRig
graphs [
 RigVMFunctionLibrary {links [], nodes []}
 RigVMModel {nodes [
   RigUnit_BeginExecution = &RigUnit_BeginExecution()
   n1 = &RigVMFunction_Sequence(ExecuteContext=@RigUnit_BeginExecution.ExecutePin)
   /Script/ControlRig.RigUnit_ChainHarmonicsPerItem(
    ChainRoot=(Type=Bone,Name="dyn_hair_curl_02_lf")
    DrawWorldOffset=(Rotation=(X=0,Y=0,Z=0,W=1),Translation=(X=0,Y=0,Z=0),Scale3D=(X=1,Y=1,Z=1))
    Pendulum=(bEnabled=false,PendulumStiffness=50,PendulumGravity=(X=-900,Y=-500,Z=0),PendulumBlend=1,PendulumDrag=20,PendulumMinimum=0,PendulumMaximum=0.8,PendulumEase=SineEaseInOut,UnwindAxis=(X=0,Y=0,Z=0),UnwindMinimum=0,UnwindMaximum=0)
    Reach=(bEnabled=false,ReachTarget=(X=0,Y=0,Z=0),ReachAxis=(X=1,Y=1,Z=1),ReachMinimum=0,ReachMaximum=0,ReachEase=SineEaseInOut)
    Speed=(X=0,Y=0,Z=0.3)
    Wave=(bEnabled=True,WaveFrequency=(X=0,Y=0,Z=1),WaveAmplitude=(X=0,Y=0,Z=1),WaveOffset=(X=0,Y=0,Z=1),WaveNoise=(X=0,Y=0,Z=0),WaveMinimum=0.1,WaveMaximum=1,WaveEase=SineEaseInOut)
    WaveCurve="(EditorCurveData=(Keys=((InterpMode=RCIM_Linear,TangentMode=RCTM_Auto,TangentWeightMode=RCTWM_WeightedNone,Time=0,Value=0.2,ArriveTangent=0,ArriveTangentWeight=0,LeaveTangent=0,LeaveTangentWeight=0),(InterpMode=RCIM_Linear,TangentMode=RCTM_Auto,TangentWeig ...(truncated)"
    bDrawDebug=True
    ExecutePin=@n1.A
   ) {is=struct RigVMUnitNode}
   > &RigUnit_ChainHarmonicsPerItem(
    ChainRoot=(Type=Bone,Name="dyn_hair_curl_02_rt")
    DrawWorldOffset=(Rotation=(X=0,Y=0,Z=0,W=1),Translation=(X=0,Y=0,Z=0),Scale3D=(X=1,Y=1,Z=1))
    Pendulum=(bEnabled=false,PendulumStiffness=50,PendulumGravity=(X=-900,Y=-500,Z=0),PendulumBlend=1,PendulumDrag=20,PendulumMinimum=0,PendulumMaximum=0.8,PendulumEase=SineEaseInOut,UnwindAxis=(X=0,Y=0,Z=0),UnwindMinimum=0,UnwindMaximum=0)
    Reach=(bEnabled=false,ReachTarget=(X=0,Y=0,Z=0),ReachAxis=(X=1,Y=1,Z=1),ReachMinimum=0,ReachMaximum=0,ReachEase=SineEaseInOut)
    Speed=(X=0,Y=1,Z=1)
    Wave=(bEnabled=True,WaveFrequency=(X=0,Y=10,Z=0),WaveAmplitude=(X=0,Y=0.6,Z=0),WaveOffset=(X=0,Y=2,Z=0),WaveNoise=(X=0,Y=0,Z=0),WaveMinimum=0.1,WaveMaximum=1,WaveEase=SineEaseInOut)
    WaveCurve="(EditorCurveData=(Keys=((InterpMode=RCIM_Linear,TangentMode=RCTM_Auto,TangentWeightMode=RCTWM_WeightedNone,Time=0,Value=0.2,ArriveTangent=0,ArriveTangentWeight=0,LeaveTangent=0,LeaveTangentWeight=0),(InterpMode=RCIM_Linear,TangentMode=RCTM_Auto,TangentWeig ...(truncated)"
    bDrawDebug=True
   )
   > &RigUnit_ChainHarmonicsPerItem(
    ChainRoot=(Type=Bone,Name="dyn_hair_curl_03_rt")
    DrawWorldOffset=(Rotation=(X=0,Y=0,Z=0,W=1),Translation=(X=0,Y=0,Z=0),Scale3D=(X=1,Y=1,Z=1))
    Pendulum=(bEnabled=false,PendulumStiffness=50,PendulumGravity=(X=-900,Y=-500,Z=0),PendulumBlend=1,PendulumDrag=0.98,PendulumMinimum=0,PendulumMaximum=0.9,PendulumEase=SineEaseInOut,UnwindAxis=(X=0,Y=0,Z=0),UnwindMinimum=0,UnwindMaximum=0)
    Reach=(bEnabled=false,ReachTarget=(X=0,Y=0,Z=0),ReachAxis=(X=1,Y=1,Z=1),ReachMinimum=0,ReachMaximum=0,ReachEase=SineEaseInOut)
    Speed=(X=0,Y=1,Z=1)
    Wave=(bEnabled=True,WaveFrequency=(X=0,Y=1,Z=1),WaveAmplitude=(X=0,Y=1,Z=1),WaveOffset=(X=2,Y=0,Z=2),WaveNoise=(X=0,Y=0,Z=0),WaveMinimum=-1,WaveMaximum=1,WaveEase=SineEaseInOut)
    WaveCurve="(EditorCurveData=(Keys=((InterpMode=RCIM_Linear,TangentMode=RCTM_Auto,TangentWeightMode=RCTWM_WeightedNone,Time=0,Value=0.2,ArriveTangent=0,ArriveTangentWeight=0,LeaveTangent=0,LeaveTangentWeight=0),(InterpMode=RCIM_Linear,TangentMode=RCTM_Auto,TangentWeig ...(truncated)"
    bDrawDebug=True
   )
   > &RigUnit_ChainHarmonicsPerItem(
    ChainRoot=(Type=Bone,Name="dyn_hair_curl_04_rt")
    DrawWorldOffset=(Rotation=(X=0,Y=0,Z=0,W=1),Translation=(X=0,Y=0,Z=0),Scale3D=(X=1,Y=1,Z=1))
    Pendulum=(bEnabled=false,PendulumStiffness=50,PendulumGravity=(X=-900,Y=-500,Z=0),PendulumBlend=1,PendulumDrag=0.98,PendulumMinimum=0,PendulumMaximum=0.9,PendulumEase=SineEaseInOut,UnwindAxis=(X=0,Y=0,Z=0),UnwindMinimum=0,UnwindMaximum=0)
    Reach=(bEnabled=false,ReachTarget=(X=0,Y=0,Z=0),ReachAxis=(X=1,Y=1,Z=1),ReachMinimum=0,ReachMaximum=0,ReachEase=SineEaseInOut)
    Speed=(X=0,Y=1,Z=1)
    Wave=(bEnabled=True,WaveFrequency=(X=1,Y=1,Z=1),WaveAmplitude=(X=1,Y=1,Z=1),WaveOffset=(X=2,Y=0,Z=2),WaveNoise=(X=0,Y=0,Z=0),WaveMinimum=-1,WaveMaximum=1,WaveEase=SineEaseInOut)
    WaveCurve="(EditorCurveData=(Keys=((InterpMode=RCIM_Linear,TangentMode=RCTM_Auto,TangentWeightMode=RCTWM_WeightedNone,Time=0,Value=0.2,ArriveTangent=0,ArriveTangentWeight=0,LeaveTangent=0,LeaveTangentWeight=0),(InterpMode=RCIM_Linear,TangentMode=RCTM_Auto,TangentWeig ...(truncated)"
    bDrawDebug=True
   )
   > &RigUnit_ChainHarmonicsPerItem(
    ChainRoot=(Type=Bone,Name="dyn_hair_curl_02_cr")
    DrawWorldOffset=(Rotation=(X=0,Y=0,Z=0,W=1),Translation=(X=0,Y=0,Z=0),Scale3D=(X=1,Y=1,Z=1))
    Pendulum=(bEnabled=false,PendulumStiffness=50,PendulumGravity=(X=-900,Y=-500,Z=0),PendulumBlend=1,PendulumDrag=0.98,PendulumMinimum=0,PendulumMaximum=0.9,PendulumEase=SineEaseInOut,UnwindAxis=(X=0,Y=0,Z=0),UnwindMinimum=0,UnwindMaximum=0)
    Reach=(bEnabled=false,ReachTarget=(X=0,Y=0,Z=0),ReachAxis=(X=1,Y=1,Z=1),ReachMinimum=0,ReachMaximum=0,ReachEase=SineEaseInOut)
    Speed=(X=0,Y=0,Z=1)
    Wave=(bEnabled=True,WaveFrequency=(X=0,Y=0,Z=0.4),WaveAmplitude=(X=0,Y=0,Z=0.2),WaveOffset=(X=0,Y=0,Z=0.4),WaveNoise=(X=0,Y=0,Z=0),WaveMinimum=0,WaveMaximum=1,WaveEase=SineEaseInOut)
    WaveCurve=(EditorCurveData=(Keys=((),(Time=1,Value=-1))))
    bDrawDebug=True
   )
   > &RigUnit_ChainHarmonicsPerItem(
    ChainRoot=(Type=Bone,Name="dyn_hair_04_rt")
    DrawWorldOffset=(Rotation=(X=0,Y=0,Z=0,W=1),Translation=(X=0,Y=0,Z=0),Scale3D=(X=1,Y=1,Z=1))
    Pendulum=(bEnabled=false,PendulumStiffness=50,PendulumGravity=(X=-900,Y=-500,Z=0),PendulumBlend=1,PendulumDrag=0.98,PendulumMinimum=0,PendulumMaximum=0.9,PendulumEase=SineEaseInOut,UnwindAxis=(X=0,Y=0,Z=0),UnwindMinimum=0,UnwindMaximum=0)
    Reach=(bEnabled=false,ReachTarget=(X=0,Y=0,Z=0),ReachAxis=(X=1,Y=1,Z=1),ReachMinimum=0,ReachMaximum=0,ReachEase=SineEaseInOut)
    Speed=(X=0,Y=0,Z=1)
    Wave=(bEnabled=True,WaveFrequency=(X=0,Y=0,Z=1),WaveAmplitude=(X=0,Y=0,Z=1),WaveOffset=(X=0,Y=0,Z=1),WaveNoise=(X=0,Y=0,Z=0),WaveMinimum=-0.5,WaveMaximum=0.5,WaveEase=SineEaseInOut)
    WaveCurve=(EditorCurveData=(Keys=((Value=0.2),(Time=5,Value=0.5)),DefaultValue=340282346638528859811704183484516925440,PreInfinityExtrap=RCCE_Constant,PostInfinityExtrap=RCCE_Constant),ExternalCurve=None)
    bDrawDebug=True
   )
   > &RigUnit_ChainHarmonicsPerItem(
    ChainRoot=(Type=Bone,Name="dyn_hair_05_cr")
    DrawWorldOffset=(Rotation=(X=0,Y=0,Z=0,W=1),Translation=(X=0,Y=0,Z=0),Scale3D=(X=1,Y=1,Z=1))
    Pendulum=(bEnabled=false,PendulumStiffness=50,PendulumGravity=(X=-900,Y=-500,Z=0),PendulumBlend=1,PendulumDrag=0.98,PendulumMinimum=0,PendulumMaximum=0.9,PendulumEase=SineEaseInOut,UnwindAxis=(X=0,Y=0,Z=0),UnwindMinimum=0,UnwindMaximum=0)
    Reach=(bEnabled=false,ReachTarget=(X=0,Y=0,Z=0),ReachAxis=(X=1,Y=1,Z=1),ReachMinimum=0,ReachMaximum=0,ReachEase=SineEaseInOut)
    Speed=(X=0,Y=0,Z=1)
    Wave=(bEnabled=True,WaveFrequency=(X=0,Y=0,Z=1),WaveAmplitude=(X=0,Y=0,Z=1),WaveOffset=(X=0,Y=0,Z=1),WaveNoise=(X=0,Y=0,Z=0),WaveMinimum=-0.5,WaveMaximum=0.5,WaveEase=SineEaseInOut)
    WaveCurve=(EditorCurveData=(Keys=((Value=0.2),(Time=5,Value=0.5)),DefaultValue=340282346638528859811704183484516925440,PreInfinityExtrap=RCCE_Constant,PostInfinityExtrap=RCCE_Constant),ExternalCurve=None)
    bDrawDebug=True
   )
   RigVMFunction_MathQuaternionToEuler = &RigVMFunction_MathQuaternionToEuler(
    RotationOrder=ZYX
    Value=&RigUnit_GetRelativeTransformForItem(Child=(Type=Bone,Name="aim_hair_rt"), Parent=(Type=Bone,Name="base_hair_rt"), bChildInitial=False, bParentInitial=False).RelativeTransform.Rotation
   )
   Remap = &RigVMFunction_MathDoubleRemap(SourceMaximum=-40, SourceMinimum=0, TargetMaximum=1, TargetMinimum=0, Value=@RigVMFunction_MathQuaternionToEuler.Result.Y)
   &RigUnit_ModifyTransforms(
    ItemToModify=((Item=(Type=Bone,Name="crown_rt"),Transform=(Rotation=(X=0,Y=-0.2588,Z=0,W=0.9659),Translation=(X=0,Y=0,Z=0),Scale3D=(X=1,Y=1,Z=1))))
    WeightMinimum=0
    Weight=@Remap.Result
    ExecutePin=@n1.B
   )
   n2 = &RigUnit_GetRelativeTransformForItem(Child=(Type=Bone,Name="aim_hair_lf"), Parent=(Type=Bone,Name="base_hair_lf"), bChildInitial=False, bParentInitial=False)
   > &RigUnit_ModifyTransforms(
    ItemToModify=((Item=(Type=Bone,Name="crown_lf"),Transform=(Rotation=(X=0,Y=-0.2588,Z=0,W=0.9659),Translation=(X=0,Y=0,Z=0),Scale3D=(X=1,Y=1,Z=1))))
    WeightMinimum=0
    Weight=&RigVMFunction_MathDoubleRemap(
     SourceMaximum=-40
     SourceMinimum=0
     TargetMaximum=1
     TargetMinimum=0
     Value=&RigVMFunction_MathQuaternionToEuler(RotationOrder=ZYX, Value=@n2.RelativeTransform.Rotation).Result.Y
    ).Result
   )
   &RigUnit_ChainHarmonicsPerItem(
    ChainRoot=(Type=Bone,Name="None")
    DrawWorldOffset=(Rotation=(X=0,Y=0,Z=0,W=1),Translation=(X=0,Y=0,Z=0),Scale3D=(X=1,Y=1,Z=1))
    ExecutePin=()
    Pendulum=(bEnabled=False,PendulumStiffness=2,PendulumGravity=(X=0,Y=0,Z=0),PendulumBlend=0.75,PendulumDrag=0.98,PendulumMinimum=0,PendulumMaximum=0.1,PendulumEase=Linear,UnwindAxis=(X=0,Y=1,Z=0),UnwindMinimum=0.2,UnwindMaximum=0.05)
    Reach=(bEnabled=False,ReachTarget=(X=0,Y=0,Z=0),ReachAxis=(X=1,Y=0,Z=0),ReachMinimum=0,ReachMaximum=0,ReachEase=Linear)
    Speed=(X=1,Y=1,Z=1)
    Wave=(bEnabled=True,WaveFrequency=(X=1,Y=0.6,Z=0.8),WaveAmplitude=(X=0,Y=1,Z=0),WaveOffset=(X=0,Y=1,Z=2),WaveNoise=(X=0,Y=0,Z=0),WaveMinimum=0,WaveMaximum=1,WaveEase=Linear)
    WaveCurve="(EditorCurveData=(Keys=((InterpMode=RCIM_Linear,TangentMode=RCTM_Auto,TangentWeightMode=RCTWM_WeightedNone,Time=0,Value=0,ArriveTangent=0,ArriveTangentWeight=0,LeaveTangent=0,LeaveTangentWeight=0),(InterpMode=RCIM_Linear,TangentMode=RCTM_Auto,TangentWeight ...(truncated)"
    bDrawDebug=True
   )
   &RigUnit_ChainHarmonicsPerItem(
    ChainRoot=(Type=Bone,Name="dyn_hair_curl_04_lf")
    DrawWorldOffset=(Rotation=(X=0,Y=0,Z=0,W=1),Translation=(X=0,Y=0,Z=0),Scale3D=(X=1,Y=1,Z=1))
    ExecutePin=()
    Pendulum=(bEnabled=false,PendulumStiffness=50,PendulumGravity=(X=-900,Y=-500,Z=0),PendulumBlend=1,PendulumDrag=0.98,PendulumMinimum=0,PendulumMaximum=0.9,PendulumEase=SineEaseInOut,UnwindAxis=(X=0,Y=0,Z=0),UnwindMinimum=0,UnwindMaximum=0)
    Reach=(bEnabled=false,ReachTarget=(X=0,Y=0,Z=0),ReachAxis=(X=1,Y=1,Z=1),ReachMinimum=0,ReachMaximum=0,ReachEase=SineEaseInOut)
    Speed=(X=0,Y=1,Z=0)
    Wave=(bEnabled=True,WaveFrequency=(X=2,Y=0,Z=2),WaveAmplitude=(X=2,Y=0,Z=2),WaveOffset=(X=2,Y=0,Z=2),WaveNoise=(X=0,Y=0,Z=0),WaveMinimum=-1,WaveMaximum=1,WaveEase=SineEaseInOut)
    WaveCurve=(EditorCurveData=(Keys=((Value=0.2),(Time=5,Value=0.5)),DefaultValue=340282346638528859811704183484516925440,PreInfinityExtrap=RCCE_Constant,PostInfinityExtrap=RCCE_Constant),ExternalCurve=None)
    bDrawDebug=True
   )
   &RigUnit_ModifyTransforms(
    ExecutePin=()
    ItemToModify="((Item=(Type=Bone,Name=\"dyn_hair_curl_01_cr\"),Transform=(Rotation=(X=0,Y=-0.0872,Z=0,W=0.9962),Translation=(X=0,Y=0,Z=0),Scale3D=(X=1,Y=1,Z=1))),(Item=(Type=Bone,Name=\"dyn_hair_curl_02_lf\"),Transform=(Rotation=(X=0,Y=-0.0872,Z=0,W=0.9962),Translation=(X=0, ...(truncated)"
    Weight=1
    WeightMinimum=0
   )
  ]}
]
hierarchy [
 root {loc=<0 0 0>, rot=<0 0 0 1>}
 pelvis {parent=root, loc=<0 -0.5301 98.0571>, rot=<0.0127 0.707 -0.0127 -0.707>}
 spine_01 {parent=pelvis, loc=<6.1208 -0.9328 0>, rot=<0 0 0.034 -0.9994>}
 spine_02 {parent=spine_01, loc=<6.5997 -0.2095 0>, rot=<0 0 0.0238 -0.9997>}
 spine_03 {parent=spine_02, loc=<6.4914 -0.1237 0>, rot=<0 0 0.0149 -0.9999>}
 spine_04 {parent=spine_03, loc=<6.9598 0.0794 0>, rot=<0 0 -0.0132 -0.9999>}
 spine_05 {parent=spine_04, loc=<6.9987 0 0>, rot=<0.0075 -0.0008 -0.0011 -1>}
 neck_01 {parent=spine_05, loc=<11.3138 0.1529 0>, rot=<0 0 -0.2003 -0.9797>}
 neck_02 {parent=neck_01, loc=<4.5627 0 0>, rot=<0 0 0.0504 -0.9987>}
 head {parent=neck_02, loc=<5.3914 0 0>, rot=<0 0 0.115 -0.9934>}
 base_hair_cr {parent=head, loc=<10.1379 -6.9188 0.0119>, rot=<0 -0.7071 0.7071 0>}
 aim_hair_cr {parent=base_hair_cr, loc=<0 0 0>, rot=<0 0 0 1>}
 offset_hair_cr {parent=aim_hair_cr, loc=<0 0 0>, rot=<0 0 0 1>}
 dyn_hair_01_cr {parent=offset_hair_cr, loc=<0 0 0>, rot=<0 0 0 1>}
 dyn_hair_02_cr {parent=dyn_hair_01_cr, loc=<13.5092 0 0>, rot=<0 0.206 0 -0.9785>}
 dyn_hair_03_cr {parent=dyn_hair_02_cr, loc=<7.0701 0 0>, rot=<0 0.0189 0 -0.9998>}
 dyn_hair_04_cr {parent=dyn_hair_03_cr, loc=<7.0629 0 0>, rot=<0 -0.2023 0 -0.9793>}
 dyn_hair_05_cr {parent=dyn_hair_04_cr, loc=<7.0605 0 0>, rot=<0 -0.1054 0 -0.9944>}
 dyn_hair_06_cr {parent=dyn_hair_05_cr, loc=<2.164 1.4022 0>, rot=<0 0 0 1>}
 dyn_hair_curl_01_cr {parent=dyn_hair_01_cr, loc=<12.467 0.14 2.493>, rot=<0.0541 0.2693 0.0381 -0.9608>}
 dyn_hair_curl_02_cr {parent=dyn_hair_curl_01_cr, loc=<6.2827 0.6481 0.2198>, rot=<0.1439 -0.0607 -0.2545 -0.9544>}
 dyn_hair_curl_03_cr {parent=dyn_hair_curl_02_cr, loc=<7.3632 0.299 0.5179>, rot=<-0.182 -0.168 -0.1471 -0.9576>}
 dyn_hair_curl_04_cr {parent=dyn_hair_curl_03_cr, loc=<8.2889 0 0>, rot=<-0.1399 0.1051 0.2077 -0.9624>}
 base_hair_rt {parent=head, loc=<9.7889 -3.6865 5.8508>, rot=<0.0598 -0.7046 0.7046 -0.0598>}
 aim_hair_rt {parent=base_hair_rt, loc=<0 0 0>, rot=<0 0 0 1>}
 offset_hair_rt {parent=aim_hair_rt, loc=<0 0 0>, rot=<0 0 0 1>}
 dyn_hair_01_rt {parent=offset_hair_rt, loc=<0 0 0>, rot=<0 0 0 1>}
 dyn_hair_02_rt {parent=dyn_hair_01_rt, loc=<13.5092 0 0>, rot=<0 0.1499 0 -0.9887>}
 dyn_hair_03_rt {parent=dyn_hair_02_rt, loc=<7.0701 0 0>, rot=<0 -0.0342 0 -0.9994>}
 dyn_hair_04_rt {parent=dyn_hair_03_rt, loc=<7.0629 0 0>, rot=<0 -0.1708 0 -0.9853>}
 dyn_hair_05_rt {parent=dyn_hair_04_rt, loc=<7.0605 0 0>, rot=<0 -0.1054 0 -0.9944>}
 dyn_hair_curl_01_rt {parent=dyn_hair_01_rt, loc=<12.805 -0.604 2.299>, rot=<-0.1649 0.0908 0.1742 -0.9665>}
 dyn_hair_curl_02_rt {parent=dyn_hair_curl_01_rt, loc=<6.9718 0 0>, rot=<-0.0137 -0.0667 -0.0429 -0.9968>}
 dyn_hair_curl_03_rt {parent=dyn_hair_curl_02_rt, loc=<6.99 0 0>, rot=<0.0067 -0.0626 -0.1744 -0.9827>}
 dyn_hair_curl_04_rt {parent=dyn_hair_curl_03_rt, loc=<6.951 0 0>, rot=<0.0452 0.0242 -0.0646 -0.9966>}
 base_hair_lf {parent=head, loc=<9.7889 -3.6865 -5.851>, rot=<0.0598 -0.7046 0.7046 -0.0598>}
 aim_hair_lf {parent=base_hair_lf, loc=<0 0 0>, rot=<0 0 0 1>}
 offset_hair_lf {parent=aim_hair_lf, loc=<0 0 0>, rot=<0 0 0 1>}
 dyn_hair_01_lf {parent=offset_hair_lf, loc=<0 0 0>, rot=<0 0 0 1>}
 dyn_hair_02_lf {parent=dyn_hair_01_lf, loc=<13.5092 0 0>, rot=<0 0.1499 0 -0.9887>}
 dyn_hair_03_lf {parent=dyn_hair_02_lf, loc=<7.0701 0 0>, rot=<0 -0.0342 0 -0.9994>}
 dyn_hair_04_lf {parent=dyn_hair_03_lf, loc=<7.0629 0 0>, rot=<0 -0.1708 0 -0.9853>}
 dyn_hair_05_lf {parent=dyn_hair_04_lf, loc=<7.0605 0 0>, rot=<0 -0.1054 0 -0.9944>}
 dyn_hair_curl_01_lf {parent=dyn_hair_01_lf, loc=<4.7046 -0.9377 2.0295>, rot=<-0.0537 0.0316 -0.0771 -0.9951>}
 dyn_hair_curl_02_lf {parent=dyn_hair_curl_01_lf, loc=<8.745 0.534 0>, rot=<0.0789 0.0054 -0.068 -0.9945>}
 dyn_hair_curl_03_lf {parent=dyn_hair_curl_02_lf, loc=<6.68 0.186 0>, rot=<0 0 0.1008 -0.9949>}
 dyn_hair_curl_04_lf {parent=dyn_hair_curl_03_lf, loc=<8.1662 0 0>, rot=<-0.0213 -0.118 0.1764 -0.977>}
 dyn_hair_curl_05_lf {parent=dyn_hair_curl_04_lf, loc=<6.646 0 0>, rot=<0 0 -0.1017 -0.9948>}
 base_hair_outer_rt {parent=head, loc=<7.0919 -2.522 7.6635>, rot=<0.0558 -0.7049 0.7049 -0.0558>}
 aim_hair_outer_rt {parent=base_hair_outer_rt, loc=<0 0 0>, rot=<0 0 0 1>}
 offset_hair_outer_rt {parent=aim_hair_outer_rt, loc=<0 0 0>, rot=<0 0 0 1>}
 dyn_hair_outer_01_rt {parent=offset_hair_outer_rt, loc=<0 0 0>, rot=<0 0 0 1>}
 dyn_hair_outer_02_rt {parent=dyn_hair_outer_01_rt, loc=<10.9807 0 0>, rot=<0 0.151 0 -0.9885>}
 dyn_hair_outer_03_rt {parent=dyn_hair_outer_02_rt, loc=<4.5134 0 -0.0002>, rot=<0 0 0.2756 -0.9613>}
 dyn_hair_outer_04_rt {parent=dyn_hair_outer_03_rt, loc=<5.9598 -0.0001 0.0001>, rot=<-0.0047 0.137 -0.034 -0.99>}
 dyn_hair_outer_05_rt {parent=dyn_hair_outer_04_rt, loc=<2.8727 0 0>, rot=<0 0 0 1>}
 base_hair_outer_lf {parent=head, loc=<7.0917 -2.522 -7.6636>, rot=<0.0558 -0.7049 0.7049 -0.0558>}
 aim_hair_outer_lf {parent=base_hair_outer_lf, loc=<0 0 0>, rot=<0 0 0 1>}
 offset_hair_outer_lf {parent=aim_hair_outer_lf, loc=<0 0 0>, rot=<0 0 0 1>}
 dyn_hair_outer_01_lf {parent=offset_hair_outer_lf, loc=<0 0 0>, rot=<0 0 0 1>}
 dyn_hair_outer_02_lf {parent=dyn_hair_outer_01_lf, loc=<10.981 0 0>, rot=<0 0.151 0 -0.9885>}
 dyn_hair_outer_03_lf {parent=dyn_hair_outer_02_lf, loc=<8.5046 0 0>, rot=<0 0 -0.2559 -0.9667>}
 dyn_hair_outer_04_lf {parent=dyn_hair_outer_03_lf, loc=<5.96 0 0>, rot=<0 0 -0.2382 -0.9712>}
 base_crown_rt {parent=head, loc=<10.71 0.923 8.511>, rot=<0.0598 -0.7046 0.7046 -0.0598>}
 crown_rt {parent=base_crown_rt, loc=<0 0 0>, rot=<0 0 0 1>}
 base_crown_lf {parent=head, loc=<10.71 0.923 -8.511>, rot=<0.0598 -0.7046 0.7046 -0.0598>}
 crown_lf {parent=base_crown_lf, loc=<0 0 0>, rot=<0 0 0 1>}
 base_hair_inner_rt {parent=head, loc=<4.2265 -1.7642 6.5374>, rot=<0.0029 -0.7071 0.7071 -0.0029>}
 aim_hair_inner_rt {parent=base_hair_inner_rt, loc=<0 0 0>, rot=<0 0 0 1>}
 offset_hair_inner_rt {parent=aim_hair_inner_rt, loc=<0 0 0>, rot=<0 0 0 1>}
 hair_inner_01_rt {parent=offset_hair_inner_rt, loc=<0 0 0>, rot=<0 0 0 1>}
 hair_inner_02_rt {parent=hair_inner_01_rt, loc=<6.829 0 0>, rot=<0 0 0 1>}
 base_hair_inner_lf {parent=head, loc=<4.227 -1.7642 -6.5374>, rot=<0.0029 -0.7071 0.7071 -0.0029>}
 aim_hair_inner_lf {parent=base_hair_inner_lf, loc=<0 0 0>, rot=<0 0 0 1>}
 offset_hair_inner_lf {parent=aim_hair_inner_lf, loc=<0 0 0>, rot=<0 0 0 1>}
 hair_inner_01_lf {parent=offset_hair_inner_lf, loc=<0 0 0>, rot=<0 0 0 1>}
 hair_inner_02_lf {parent=hair_inner_01_lf, loc=<6.8293 0 0>, rot=<0 0 0 1>}
 FACIAL_C_Jaw {parent=head, loc=<0.1816 1.8765 0.0012>, rot=<0.6238 -0.3335 -0.6228 -0.3342>}
 FACIAL_C_Jawline {parent=FACIAL_C_Jaw, loc=<-0.0083 2.9396 6.8797>, rot=<0.2586 0 0 -0.966>}
 FACIAL_L_ChinSide {parent=FACIAL_C_Jaw, loc=<2.6277 0.5162 6.8339>, rot=<-0.2208 -0.3738 0.0499 -0.8995>}
 FACIAL_R_ChinSide {parent=FACIAL_C_Jaw, loc=<-2.6365 0.504 6.8306>, rot=<-0.2208 0.3738 -0.0499 -0.8995>}
 FACIAL_C_Chin {parent=FACIAL_C_Jaw, loc=<-0.0063 1.4425 7.1859>, rot=<-0.2591 0 0 -0.9659>}
 clavicle_l {parent=spine_05, loc=<5.4831 -0.1053 -3.2946>, rot=<0.0465 -0.7289 -0.009 -0.6829>}
 upperarm_l {parent=clavicle_l, loc=<9.8658 -1.2136 -0.4105>, rot=<0.0169 -0.3401 -0.0266 -0.9398>}
 lowerarm_l {parent=upperarm_l, loc=<26.465 -0.8345 0.259>, rot=<0.0034 0.0162 -0.2128 -0.977>}
 hand_l {parent=lowerarm_l, loc=<24.6153 -0.4172 -1.3501>, rot=<-0.6854 -0.0158 0.0518 -0.7262>}
 hand_attach_l {parent=hand_l, loc=<8.5 -3 0>, rot=<0.7071 0.7071 0 0>}
 weapon_l {parent=hand_l, loc=<-0.1974 -3.4111 0.3811>, rot=<0 0 0 1>}
 thumb_01_l {parent=hand_l, loc=<1.3118 -0.3191 -2.459>, rot=<0.6546 -0.4042 0.0664 -0.6354>}
 thumb_02_l {parent=thumb_01_l, loc=<5.9336 -0.0448 0.0401>, rot=<-0.0037 0.0213 0.298 -0.9543>}
 thumb_03_l {parent=thumb_02_l, loc=<2.2223 0.3682 -0.0907>, rot=<0.0118 -0.013 0.0066 -0.9998>}
 index_metacarpal_l {parent=hand_l, loc=<1.6535 0.0183 -2.96>, rot=<0.0158 -0.0124 -0.0173 -0.9996>}
 index_01_l {parent=index_metacarpal_l, loc=<8.3008 0.0357 -0.0524>, rot=<-0.0044 0.0485 0.0924 -0.9945>}
 index_02_l {parent=index_01_l, loc=<3.3587 -0.3004 0.0452>, rot=<-0.002 0.0027 0.3225 -0.9466>}
 index_03_l {parent=index_02_l, loc=<2.5578 -0.123 -0.0252>, rot=<-0.002 0.0067 0.2558 -0.9667>}
 middle_metacarpal_l {parent=hand_l, loc=<2.1856 -0.1725 -0.5984>, rot=<0.006 0.0168 -0.0329 -0.9993>}
 middle_01_l {parent=middle_metacarpal_l, loc=<7.3275 -0.2447 0.0129>, rot=<-0.0208 0.0042 0.0858 -0.9961>}
 middle_02_l {parent=middle_01_l, loc=<4.0123 0.1401 0.1175>, rot=<-0.0218 0.0258 0.3512 -0.9357>}
 middle_03_l {parent=middle_02_l, loc=<2.6077 0.0534 0.1373>, rot=<0.0082 0.0188 0.1708 -0.9851>}
 ring_metacarpal_l {parent=hand_l, loc=<2.2074 -0.3625 1.3638>, rot=<-0.0284 0.0722 -0.0458 -0.9959>}
 ring_01_l {parent=ring_metacarpal_l, loc=<7.0295 -0.6197 -0.1446>, rot=<-0.0489 -0.0252 0.1892 -0.9804>}
 ring_02_l {parent=ring_01_l, loc=<3.6174 0.4647 -0.1027>, rot=<0 0 0.2806 -0.9598>}
 ring_03_l {parent=ring_02_l, loc=<2.4152 0.114 0.0199>, rot=<0 0 0.1189 -0.9929>}
 pinky_metacarpal_l {parent=hand_l, loc=<1.7616 -0.9753 3.3814>, rot=<-0.0397 0.0814 -0.069 -0.9935>}
 pinky_01_l {parent=pinky_metacarpal_l, loc=<7.1709 -0.4831 -0.1094>, rot=<-0.0654 -0.0035 0.2242 -0.9723>}
 pinky_02_l {parent=pinky_01_l, loc=<2.8274 0.0893 -0.1043>, rot=<0.0037 0.0011 0.2854 -0.9584>}
 pinky_03_l {parent=pinky_02_l, loc=<2.2046 0.0307 -0.0078>, rot=<-0.0002 0.0098 0.1445 -0.9895>}
 deform_wrist_dn_base_l {parent=hand_l, loc=<0 -1.5326 0>, rot=<0 0 0 -1>}
 deform_wrist_dn_l {parent=deform_wrist_dn_base_l, loc=<0 0 0>, rot=<0 0 0 1>}
 deform_wrist_up_base_l {parent=hand_l, loc=<0 2.3009 0>, rot=<0 0 0 -1>}
 deform_wrist_up_l {parent=deform_wrist_up_base_l, loc=<0 0 0>, rot=<0 0 0 1>}
 lowerarm_twist_02_l {parent=lowerarm_l, loc=<10.6707 0 0>, rot=<0 0 0 1>}
 lowerarm_twist_01_l {parent=lowerarm_l, loc=<18.316 0 0>, rot=<0 0 0 1>}
 deform_elbow_in_base_l {parent=lowerarm_l, loc=<0.1834 3.6062 -0.2572>, rot=<0 0 0 -1>}
 deform_elbow_in_l {parent=deform_elbow_in_base_l, loc=<0 0 0>, rot=<0 0 0 1>}
 deform_elbow_out_base_l {parent=lowerarm_l, loc=<0 -4 0>, rot=<0 0 0 1>}
 deform_elbow_out_l {parent=deform_elbow_out_base_l, loc=<0 0 0>, rot=<0 0 0 1>}
 upperarm_twist_01_l {parent=upperarm_l, loc=<7.2135 -0.1134 0>, rot=<0 0 0 1>}
 deform_upperarm_bk_base_l {parent=upperarm_twist_01_l, loc=<-6.4053 -4.3746 0.314>, rot=<0 0 0 -1>}
 deform_upperarm_bk_l {parent=deform_upperarm_bk_base_l, loc=<0 0 0>, rot=<0 0 0 1>}
 deform_upperarm_fr_base_l {parent=upperarm_twist_01_l, loc=<-5.6312 5.092 0.3354>, rot=<0 0 0 -1>}
 deform_upperarm_fr_l {parent=deform_upperarm_fr_base_l, loc=<0 0 0>, rot=<0 0 0 1>}
 deform_upperarm_up_base_l {parent=upperarm_twist_01_l, loc=<-8.891 0.4825 3.9517>, rot=<0 0 0 -1>}
 deform_upperarm_up_l {parent=deform_upperarm_up_base_l, loc=<0 0 0>, rot=<0 0 0 1>}
 upperarm_twist_02_l {parent=upperarm_l, loc=<19.7998 -0.1134 0>, rot=<0 0 0 1>}
 clavicle_r {parent=spine_05, loc=<5.4831 -0.1053 3.2946>, rot=<0.7289 0.0465 -0.6829 0.009>}
 upperarm_r {parent=clavicle_r, loc=<-9.8658 1.2136 0.4105>, rot=<0.0169 -0.3401 -0.0266 -0.9398>}
 lowerarm_r {parent=upperarm_r, loc=<-26.465 0.8345 -0.259>, rot=<0.0034 0.0162 -0.2128 -0.977>}
 hand_r {parent=lowerarm_r, loc=<-22.7355 0.3203 0.6106>, rot=<-0.6784 -0.0223 0.0735 -0.7307>}
 thumb_01_r {parent=hand_r, loc=<-1.3118 0.3191 2.459>, rot=<0.6546 -0.4042 0.0664 -0.6354>}
 thumb_02_r {parent=thumb_01_r, loc=<-5.9336 0.0448 -0.0401>, rot=<-0.0037 0.0213 0.298 -0.9543>}
 thumb_03_r {parent=thumb_02_r, loc=<-2.2223 -0.3682 0.0907>, rot=<0.0118 -0.013 0.0066 -0.9998>}
 index_metacarpal_r {parent=hand_r, loc=<-1.6535 -0.0183 2.96>, rot=<0.0158 -0.0124 -0.0173 -0.9996>}
 index_01_r {parent=index_metacarpal_r, loc=<-8.3008 -0.0357 0.0524>, rot=<-0.0044 0.0485 0.0924 -0.9945>}
 index_02_r {parent=index_01_r, loc=<-3.3587 0.3004 -0.0452>, rot=<-0.002 0.0027 0.3225 -0.9466>}
 index_03_r {parent=index_02_r, loc=<-2.5578 0.123 0.0252>, rot=<-0.002 0.0067 0.2558 -0.9667>}
 middle_metacarpal_r {parent=hand_r, loc=<-2.1856 0.1725 0.5984>, rot=<0.006 0.0168 -0.0329 -0.9993>}
 middle_01_r {parent=middle_metacarpal_r, loc=<-7.3275 0.2447 -0.0129>, rot=<-0.0208 0.0042 0.0858 -0.9961>}
 middle_02_r {parent=middle_01_r, loc=<-4.0123 -0.1401 -0.1175>, rot=<-0.0218 0.0258 0.3512 -0.9357>}
 middle_03_r {parent=middle_02_r, loc=<-2.6077 -0.0534 -0.1373>, rot=<0.0082 0.0188 0.1708 -0.9851>}
 ring_metacarpal_r {parent=hand_r, loc=<-2.2074 0.3625 -1.3638>, rot=<-0.0284 0.0722 -0.0458 -0.9959>}
 ring_01_r {parent=ring_metacarpal_r, loc=<-7.0295 0.6197 0.1446>, rot=<-0.0489 -0.0252 0.1892 -0.9804>}
 ring_02_r {parent=ring_01_r, loc=<-3.6174 -0.4647 0.1027>, rot=<0 0 0.2806 -0.9598>}
 ring_03_r {parent=ring_02_r, loc=<-2.4152 -0.114 -0.0199>, rot=<0 0 0.1189 -0.9929>}
 pinky_metacarpal_r {parent=hand_r, loc=<-1.7616 0.9753 -3.3814>, rot=<-0.0397 0.0814 -0.069 -0.9935>}
 pinky_01_r {parent=pinky_metacarpal_r, loc=<-7.1709 0.4831 0.1094>, rot=<-0.0654 -0.0035 0.2242 -0.9723>}
 pinky_02_r {parent=pinky_01_r, loc=<-2.8274 -0.0893 0.1043>, rot=<0.0037 0.0011 0.2854 -0.9584>}
 pinky_03_r {parent=pinky_02_r, loc=<-2.2046 -0.0307 0.0078>, rot=<-0.0002 0.0098 0.1445 -0.9895>}
 weapon_r {parent=hand_r, loc=<0.1974 3.4111 -0.3811>, rot=<0 0 0 1>}
 hand_attach_r {parent=hand_r, loc=<-8.5001 2.9999 0>, rot=<0 0 -0.7071 -0.7071>}
 deform_wrist_dn_base_r {parent=hand_r, loc=<-0.0001 1.5328 0>, rot=<0 0 0 -1>}
 deform_wrist_dn_r {parent=deform_wrist_dn_base_r, loc=<0 0 0>, rot=<0 0 0 1>}
 deform_wrist_up_base_r {parent=hand_r, loc=<-0.0002 -2.3007 -0.0001>, rot=<0 0 0 -1>}
 deform_wrist_up_r {parent=deform_wrist_up_base_r, loc=<0 0 0>, rot=<0 0 0 1>}
 lowerarm_twist_02_r {parent=lowerarm_r, loc=<-10.6707 0 0>, rot=<0 0 0 -1>}
 lowerarm_twist_01_r {parent=lowerarm_r, loc=<-18.316 0 0>, rot=<0 0 0 -1>}
 deform_elbow_in_base_r {parent=lowerarm_r, loc=<-0.1834 -3.6063 0.2572>, rot=<0 0 0 -1>}
 deform_elbow_in_r {parent=deform_elbow_in_base_r, loc=<0 0 0>, rot=<0 0 0 1>}
 deform_elbow_out_base_r {parent=lowerarm_r, loc=<0 4 -0.0001>, rot=<0 0 0 1>}
 deform_elbow_out_r {parent=deform_elbow_out_base_r, loc=<0 0 0>, rot=<0 0 0 1>}
 upperarm_twist_01_r {parent=upperarm_r, loc=<-7.2135 0.1134 0>, rot=<0 0 0 -1>}
 deform_upperarm_bk_base_r {parent=upperarm_twist_01_r, loc=<6.4052 4.3746 -0.314>, rot=<0 0 0 -1>}
 deform_upperarm_bk_r {parent=deform_upperarm_bk_base_r, loc=<0 0 0>, rot=<0 0 0 1>}
 deform_upperarm_fr_base_r {parent=upperarm_twist_01_r, loc=<5.6311 -5.092 -0.3354>, rot=<0 0 0 -1>}
 deform_upperarm_fr_r {parent=deform_upperarm_fr_base_r, loc=<0 0 0>, rot=<0 0 0 1>}
 deform_upperarm_up_base_r {parent=upperarm_twist_01_r, loc=<8.891 -0.4825 -3.9517>, rot=<0 0 0 -1>}
 deform_upperarm_up_r {parent=deform_upperarm_up_base_r, loc=<0 0 0>, rot=<0 0 0 1>}
 upperarm_twist_02_r {parent=upperarm_r, loc=<-19.7998 0.1134 0>, rot=<0 0 0 -1>}
 base_head_aim_tgt {parent=spine_05, loc=<9.0172 -16.7343 -0.0002>, rot=<0 0 0 -1>}
 head_aim_tgt {parent=base_head_aim_tgt, loc=<0 0 0>, rot=<0 0 0 1>}
 attach_backpack {parent=spine_05, loc=<-0.3547 -28.6255 -0.0095>, rot=<-0.0092 -0.7071 0.0092 -0.707>}
 attach_cape {parent=spine_05, loc=<8.1938 -21.3007 -0.0196>, rot=<0.0617 -0.7044 -0.0617 -0.7044>}
 clavicle_l_set_rot_offset {parent=spine_05, loc=<5.5072 -0.096 -3.2441>, rot=<0.0466 -0.7289 -0.009 -0.6829>}
 clavicle_l_set_rot {parent=clavicle_l_set_rot_offset, loc=<0.0155 -0.0308 -0.028>, rot=<-0.099 0.4016 -0.2631 -0.8716>}
 clavicle_r_set_rot_offset {parent=spine_05, loc=<5.4438 -0.0938 3.2306>, rot=<0.7289 0.0465 -0.683 0.0091>}
 clavicle_r_set_rot {parent=clavicle_r_set_rot_offset, loc=<-0.0148 0.0069 0.0358>, rot=<0.0545 0.3366 -0.1151 -0.933>}
 deform_pec_aim_base_l {parent=spine_04, loc=<3.2456 9.3309 -1.2237>, rot=<0.0603 -0.4033 0.3318 -0.8507>}
 deform_pec_aim_l {parent=deform_pec_aim_base_l, loc=<0 0 0>, rot=<0 0 0 1>}
 deform_pec_base_l {parent=deform_pec_aim_l, loc=<9.3665 0 0>, rot=<0 0 0 1>}
 deform_pec_l {parent=deform_pec_base_l, loc=<0 0 0>, rot=<0 0 0 1>}
 deform_pec_aim_base_r {parent=spine_04, loc=<3.2452 9.3309 1.2237>, rot=<-0.4033 -0.0603 0.8507 0.3318>}
 deform_pec_aim_r {parent=deform_pec_aim_base_r, loc=<0 0 0>, rot=<0 0 0 1>}
 deform_pec_base_r {parent=deform_pec_aim_r, loc=<-9.3667 -0.0003 -0.0004>, rot=<0 0 0 1>}
 deform_pec_r {parent=deform_pec_base_r, loc=<0 0 0>, rot=<0 0 0 1>}
 base_latHelp_lf {parent=spine_04, loc=<1.8113 -0.8012 -9.132>, rot=<0.0294 -0.7065 -0.0294 -0.7065>}
 Deform_latHelp_lf {parent=base_latHelp_lf, loc=<0 0 0>, rot=<0 0 0 1>}
 base_latHelp_rt {parent=spine_04, loc=<1.8113 -0.8012 9.132>, rot=<0.7065 0.0294 -0.7065 0.0294>}
 Deform_latHelp_rt {parent=base_latHelp_rt, loc=<0 0 0>, rot=<0 0 0 1>}
 thigh_l {parent=pelvis, loc=<-1.0855 -0.2109 -8.7053>, rot=<0.0009 0.0355 0.0155 -0.9993>}
 calf_l {parent=thigh_l, loc=<-42.5284 -0.0019 -0.7297>, rot=<0.0006 -0.0161 -0.0391 -0.9991>}
 foot_l {parent=calf_l, loc=<-46.7401 -2.5014 -0.592>, rot=<0.0427 -0.0238 0.0353 -0.9982>}
 ball_l {parent=foot_l, loc=<-3.5658 10.8066 -0.1544>, rot=<0.0001 -0.0001 -0.7063 -0.7079>}
 Retargetball_lf {parent=foot_l, loc=<-7.2321 11.3739 -0.945>, rot=<0.0001 -0.0001 -0.7063 -0.7079>}
 Base_FootPush_lf {parent=foot_l, loc=<0 0 0>, rot=<0 0 0 1>}
 FootPush_lf {parent=Base_FootPush_lf, loc=<0 0 0>, rot=<0 0 0 1>}
 calf_twist_02_l {parent=calf_l, loc=<-19.1546 -1.7943 -0.2638>, rot=<0.0237 0.0095 -0.0035 -0.9997>}
 calf_twist_01_l {parent=calf_l, loc=<-35.8483 -1.9197 -0.5794>, rot=<0.0237 0.0095 -0.0035 -0.9997>}
 deform_knee_in_base_l {parent=calf_l, loc=<0.8031 -5.0231 0.2456>, rot=<0 0 0 -1>}
 deform_knee_in_l {parent=deform_knee_in_base_l, loc=<0 0 0>, rot=<0 0 0 1>}
 deform_knee_out_base_l {parent=calf_l, loc=<0 4 0>, rot=<0 0 0 -1>}
 deform_knee_out_l {parent=deform_knee_out_base_l, loc=<0 0 0>, rot=<0 0 0 1>}
 thigh_twist_01_l {parent=thigh_l, loc=<-27.952 0 0>, rot=<0.0013 0 0 -1>}
 deform_groin_base_l {parent=thigh_l, loc=<0.4801 5.6776 -0.2486>, rot=<0 0 0 -1>}
 deform_groin_l {parent=deform_groin_base_l, loc=<0 0 0>, rot=<0 0 0 1>}
 thigh_r {parent=pelvis, loc=<-1.0855 -0.2109 8.7053>, rot=<0.0355 -0.0009 0.9993 0.0155>}
 calf_r {parent=thigh_r, loc=<42.5284 0.0019 0.7297>, rot=<0.0006 -0.0161 -0.0391 -0.9991>}
 foot_r {parent=calf_r, loc=<46.7401 2.5014 0.592>, rot=<0.0427 -0.0238 0.0353 -0.9982>}
 ball_r {parent=foot_r, loc=<3.5658 -10.8066 0.1544>, rot=<0.0001 -0.0001 -0.7063 -0.7079>}
 Retargetball_rt {parent=foot_r, loc=<7.0812 -10.2906 0.7382>, rot=<0.0001 -0.0001 -0.7063 -0.7079>}
 Base_FootPush_rt {parent=foot_r, loc=<0 0 0>, rot=<0 0 0 -1>}
 FootPush_rt {parent=Base_FootPush_rt, loc=<0 0 0>, rot=<0 0 0 1>}
 calf_twist_02_r {parent=calf_r, loc=<19.1546 1.7943 0.2638>, rot=<0.0237 0.0095 -0.0035 -0.9997>}
 calf_twist_01_r {parent=calf_r, loc=<35.8483 1.9197 0.5794>, rot=<0.0237 0.0095 -0.0035 -0.9997>}
 deform_knee_in_base_r {parent=calf_r, loc=<-0.8031 5.0231 -0.2457>, rot=<0 0 0 -1>}
 deform_knee_in_r {parent=deform_knee_in_base_r, loc=<0 0 0>, rot=<0 0 0 1>}
 deform_knee_out_base_r {parent=calf_r, loc=<0 -4 0>, rot=<0 0 0 -1>}
 deform_knee_out_r {parent=deform_knee_out_base_r, loc=<0 0 0>, rot=<0 0 0 1>}
 thigh_twist_01_r {parent=thigh_r, loc=<27.952 0 0>, rot=<0.0013 0 0 -1>}
 deform_groin_base_r {parent=thigh_r, loc=<-0.48 -5.6776 0.2486>, rot=<0 0 0 -1>}
 deform_groin_r {parent=deform_groin_base_r, loc=<0 0 0>, rot=<0 0 0 1>}
 deform_glute_base_l {parent=pelvis, loc=<-7.3466 -9.0213 -9.1245>, rot=<0.0172 0.031 -0.4477 -0.8935>}
 deform_glute_l {parent=deform_glute_base_l, loc=<0 0 0>, rot=<0 0 0 1>}
 deform_glute_base_r {parent=pelvis, loc=<-7.3464 -9.0213 9.1245>, rot=<0.031 -0.0172 0.8935 -0.4477>}
 deform_glute_r {parent=deform_glute_base_r, loc=<0 0 0>, rot=<0 0 0 1>}
 attach {parent=root, loc=<0 0 0>, rot=<0 0 0 1>}
 ik_hand_root {parent=root, loc=<0 0 0>, rot=<0 0 0 1>}
 ik_hand_gun {parent=ik_hand_root, loc=<-48.1743 3.3545 104.4381>, rot=<0.6276 0.1309 0.4183 -0.6434>}
 ik_hand_r {parent=ik_hand_gun, loc=<0 0 0>, rot=<0 0 0 1>}
 ik_hand_l {parent=ik_hand_gun, loc=<59.3271 67.6997 34.3524>, rot=<-0.6158 -0.7027 -0.3565 0>}
 ik_foot_root {parent=root, loc=<0 0 0>, rot=<0 0 0 1>}
 ik_foot_l {parent=ik_foot_root, loc=<14.8262 -7.3545 8.296>, rot=<0.0357 0.7045 0.0284 -0.7082>}
 ik_foot_r {parent=ik_foot_root, loc=<-14.8262 -7.3545 8.296>, rot=<0.7082 0.0284 -0.7045 0.0357>}
 root_fp {parent=root, loc=<0 0 0>, rot=<0 0 0 1>}
 attach_fp {parent=root_fp, loc=<0 0 0>, rot=<0 0 0 1>}
 camera_root {parent=root_fp, loc=<0 0 0>, rot=<0 0 0 1>}
 camera {parent=camera_root, loc=<0 0 0>, rot=<0 0 0 1>}
 attach_emote_01 {parent=root, loc=<0 0 0>, rot=<0 0 0 1>}
 attach_emote_02 {parent=root, loc=<0 0 0>, rot=<0 0 0 1>}
 attach_emote_03 {parent=root, loc=<0 0 0>, rot=<0 0 0 1>}
 warp_point_01 {parent=root, loc=<0 0 0>, rot=<0 0 0 1>}
 warp_point_02 {parent=root, loc=<0 0 0>, rot=<0 0 0 1>}
 warp_point_03 {parent=root, loc=<0 0 0>, rot=<0 0 0 1>}
]
variables [
 AlphaMov {type=double, default=0.6}
 ChainHarmonicsPerItem_Wave {
  type=FRigUnit_ChainHarmonics_Wave
  default=(bEnabled=True,WaveFrequency=(X=1,Y=5,Z=1),WaveAmplitude=(X=2,Y=-5,Z=1),WaveOffset=(X=1,Y=1,Z=1),WaveNoise=(X=0.2,Y=0.1,Z=0.2),WaveMinimum=-0.2,WaveMaximum=1,WaveEase=SineEaseInOut)
 }
 Wave Amplitude {type=double, default=1.5}
 WaveSpeed {type=double, default=-3}
]
