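# Anim Blueprint for the F_MED_SaltLattice face accessory, driving its hair bones with procedural motion and simulation.
# Copied parent and body-mesh poses pass through a Control Rig, RigidBody sim, weighted bone constraints and a post-sim Control Rig.
# AnimDynamics then springs the curl bones, disabled when a cape or backbling is worn.
# The EventGraph caches the body mesh and sets backbling, skydive, gravity and zipline velocity-scale variables.
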AnimBlueprint /BRCosmetics/Characters/Player/Female/Medium/Bodies/F_MED_SaltLattice/Meshes/Parts/F_MED_SaltLattice_FaceAcc_AnimBP
target_skeleton=&/F_MED_SaltLattice_FaceAcc_Skeleton
graphs [
 AnimGraph {nodes [
   "Constraint - dyn_hair_01_lf (RR)"!(ComponentPose, bAlphaBoolEnabled:bool=True, ConstraintWeights_0=0.3, ConstraintWeights_1:float=0.7) -> (Pose) {
    alphaInputType=Bool
    boneToModify {boneName=dyn_hair_01_lf}
    constraintSetup [{transformType=Rotation, targetBone {boneName=neck_02}}, {transformType=Rotation, targetBone {boneName=head}}]
    constraintWeights [1,1]
   }
   Output Pose(Result=Use cached pose 'AnimDynPose'()) {layerGroup=None}
   n1 = NOT Boolean(A=Get is Wearing cape and backpack() -> (is Wearing cape and backpack:bool))
   n2 = Anim Dynamics - Bone: dyn_hair_curl_03_cr!(
    ComponentPose=Local To Component(LocalPose=Use cached pose 'ConstraintPose'())
    bAlphaBoolEnabled:bool=@n1
    GravityOverride:Vector="0,0,-980"
    ComponentLinearAccScale:Vector="1,1,1"
    ComponentLinearVelScale:Vector="1,1,1"
    ExternalForce="0,0,0"
   ) -> (Pose) {
    alphaInputType=Bool
    angularDampingOverride=1
    angularSpringConstant=20000
    bAngularSpring=true
    bOverrideAngularDamping=true
    bUseGravityOverride=true
    gravityScale=0.35
    simulationSpace=RootRelative
    boundBone {boneName=dyn_hair_curl_03_cr}
    componentAppliedLinearAccClamp=<3000 3000 3000>
    componentLinearAccScale=<1 1 1>
    componentLinearVelScale=<1 1 1>
    gravityOverride {z=-980}
    physicsBodyDefinitions [{
      boundBone {boneName=dyn_hair_curl_03_cr}
      boxExtents=<30 30 30>
      constraintSetup {angularLimitsMax=<0 0 2>, angularLimitsMin=<0 -2 -2>, angularTarget=<1 0 0>}
      localJointOffset=<5 0 0>
     }]
    simSpaceSettings {maxAngularAcceleration=35, maxAngularVelocity=7, simSpaceAngularAlpha=1}
   }
   n3 = Anim Dynamics - Bone: dyn_hair_curl_04_cr!(
    ComponentPose=@n2
    bAlphaBoolEnabled:bool=@n1
    GravityOverride:Vector="0,0,-980"
    ComponentLinearAccScale:Vector="1,1,1"
    ComponentLinearVelScale:Vector="1,1,1"
    ExternalForce="0,0,0"
   ) -> (Pose) {
    alphaInputType=Bool
    angularDampingOverride=1
    angularSpringConstant=20000
    bAngularSpring=true
    bOverrideAngularDamping=true
    bUseGravityOverride=true
    gravityScale=0.35
    simulationSpace=RootRelative
    boundBone {boneName=dyn_hair_curl_04_cr}
    componentAppliedLinearAccClamp=<3000 3000 3000>
    componentLinearAccScale=<1 1 1>
    componentLinearVelScale=<1 1 1>
    gravityOverride {z=-980}
    physicsBodyDefinitions [{
      boundBone {boneName=dyn_hair_curl_04_cr}
      boxExtents=<30 30 30>
      constraintSetup {angularLimitsMax=<0 0 2>, angularLimitsMin=<0 -2 -2>, angularTarget=<1 0 0>}
      localJointOffset=<5 0 0>
     }]
    simSpaceSettings {maxAngularAcceleration=35, maxAngularVelocity=7, simSpaceAngularAlpha=1}
   }
   n4 = Anim Dynamics - Bone: dyn_hair_curl_03_lf!(
    ComponentPose=@n3
    bAlphaBoolEnabled:bool=@n1
    GravityOverride:Vector="0,0,-980"
    ComponentLinearAccScale:Vector="1,1,1"
    ComponentLinearVelScale:Vector="1,1,1"
    ExternalForce="0,0,0"
   ) -> (Pose) {
    alphaInputType=Bool
    angularDampingOverride=1
    angularSpringConstant=20000
    bAngularSpring=true
    bOverrideAngularDamping=true
    bUseGravityOverride=true
    gravityScale=0.35
    simulationSpace=RootRelative
    boundBone {boneName=dyn_hair_curl_03_lf}
    componentAppliedLinearAccClamp=<3000 3000 3000>
    componentLinearAccScale=<1 1 1>
    componentLinearVelScale=<1 1 1>
    gravityOverride {z=-980}
    physicsBodyDefinitions [{
      boundBone {boneName=dyn_hair_curl_03_lf}
      boxExtents=<30 30 30>
      constraintSetup {angularLimitsMax=<0 0 2>, angularLimitsMin=<0 -2 -2>, angularTarget=<1 0 0>}
      localJointOffset=<5 0 0>
     }]
    simSpaceSettings {maxAngularAcceleration=35, maxAngularVelocity=7, simSpaceAngularAlpha=1}
   }
   n5 = Anim Dynamics - Bone: dyn_hair_curl_04_lf!(
    ComponentPose=@n4
    bAlphaBoolEnabled:bool=@n1
    GravityOverride:Vector="0,0,-980"
    ComponentLinearAccScale:Vector="1,1,1"
    ComponentLinearVelScale:Vector="1,1,1"
    ExternalForce="0,0,0"
   ) -> (Pose) {
    alphaInputType=Bool
    angularDampingOverride=1
    angularSpringConstant=20000
    bAngularSpring=true
    bOverrideAngularDamping=true
    bUseGravityOverride=true
    gravityScale=0.35
    simulationSpace=RootRelative
    boundBone {boneName=dyn_hair_curl_04_lf}
    componentAppliedLinearAccClamp=<3000 3000 3000>
    componentLinearAccScale=<1 1 1>
    componentLinearVelScale=<1 1 1>
    gravityOverride {z=-980}
    physicsBodyDefinitions [{
      boundBone {boneName=dyn_hair_curl_04_lf}
      boxExtents=<30 30 30>
      constraintSetup {angularLimitsMax=<0 0 2>, angularLimitsMin=<0 -2 -2>, angularTarget=<1 0 0>}
      localJointOffset=<5 0 0>
     }]
    simSpaceSettings {maxAngularAcceleration=35, maxAngularVelocity=7, simSpaceAngularAlpha=1}
   }
   n6 = Anim Dynamics - Bone: dyn_hair_curl_05_lf!(
    ComponentPose=@n5
    bAlphaBoolEnabled:bool=@n1
    GravityOverride:Vector="0,0,-980"
    ComponentLinearAccScale:Vector="1,1,1"
    ComponentLinearVelScale:Vector="1,1,1"
    ExternalForce="0,0,0"
   ) -> (Pose) {
    alphaInputType=Bool
    angularDampingOverride=1
    angularSpringConstant=20000
    bAngularSpring=true
    bOverrideAngularDamping=true
    bUseGravityOverride=true
    gravityScale=0.35
    simulationSpace=RootRelative
    boundBone {boneName=dyn_hair_curl_05_lf}
    componentAppliedLinearAccClamp=<3000 3000 3000>
    componentLinearAccScale=<1 1 1>
    componentLinearVelScale=<1 1 1>
    gravityOverride {z=-980}
    physicsBodyDefinitions [{
      boundBone {boneName=dyn_hair_curl_05_lf}
      boxExtents=<30 30 30>
      constraintSetup {angularLimitsMax=<0 0 5>, angularLimitsMin=<0 -2 -5>, angularTarget=<1 0 0>}
      localJointOffset=<5 0 0>
     }]
    simSpaceSettings {maxAngularAcceleration=35, maxAngularVelocity=7, simSpaceAngularAlpha=1}
   }
   n7 = Anim Dynamics - Bone: dyn_hair_curl_03_rt!(
    ComponentPose=@n6
    bAlphaBoolEnabled:bool=@n1
    GravityOverride:Vector="0,0,-980"
    ComponentLinearAccScale:Vector="1,1,1"
    ComponentLinearVelScale:Vector="1,1,1"
    ExternalForce="0,0,0"
   ) -> (Pose) {
    alphaInputType=Bool
    angularDampingOverride=1
    angularSpringConstant=20000
    bAngularSpring=true
    bOverrideAngularDamping=true
    bUseGravityOverride=true
    gravityScale=0.35
    simulationSpace=RootRelative
    boundBone {boneName=dyn_hair_curl_03_rt}
    componentAppliedLinearAccClamp=<3000 3000 3000>
    componentLinearAccScale=<1 1 1>
    componentLinearVelScale=<1 1 1>
    gravityOverride {z=-980}
    physicsBodyDefinitions [{
      boundBone {boneName=dyn_hair_curl_03_rt}
      boxExtents=<30 30 30>
      constraintSetup {angularLimitsMax=<0 0 2>, angularLimitsMin=<0 -2 -2>, angularTarget=<1 0 0>}
      localJointOffset=<5 0 0>
     }]
    simSpaceSettings {maxAngularAcceleration=35, maxAngularVelocity=7, simSpaceAngularAlpha=1}
   }
   n8 = Anim Dynamics - Bone: dyn_hair_curl_04_rt!(
    ComponentPose=@n7
    bAlphaBoolEnabled:bool=@n1
    GravityOverride:Vector="0,0,-980"
    ComponentLinearAccScale:Vector="1,1,1"
    ComponentLinearVelScale:Vector="1,1,1"
    ExternalForce="0,0,0"
   ) -> (Pose) {
    alphaInputType=Bool
    angularDampingOverride=1
    angularSpringConstant=20000
    bAngularSpring=true
    bOverrideAngularDamping=true
    bUseGravityOverride=true
    gravityScale=0.35
    simulationSpace=RootRelative
    boundBone {boneName=dyn_hair_curl_04_rt}
    componentAppliedLinearAccClamp=<3000 3000 3000>
    componentLinearAccScale=<1 1 1>
    componentLinearVelScale=<1 1 1>
    gravityOverride {z=-980}
    physicsBodyDefinitions [{
      boundBone {boneName=dyn_hair_curl_04_rt}
      boxExtents=<30 30 30>
      constraintSetup {angularLimitsMax=<0 0 5>, angularLimitsMin=<0 -2 -5>, angularTarget=<1 0 0>}
      localJointOffset=<5 0 0>
     }]
    simSpaceSettings {maxAngularAcceleration=35, maxAngularVelocity=7, simSpaceAngularAlpha=1}
   }
   n9 = "Transform (Modify) Bone - Bone: dyn_hair_curl_02_cr"!(ComponentPose=@n8, bAlphaBoolEnabled:bool=True, Translation="0,0,0", Rotation="8.679,-30.375,-5.557", Scale="1,1,1") -> (Pose) {
    alphaInputType=Bool
    rotationMode=BMM_Additive
    rotationSpace=BCS_BoneSpace
    alphaBoolBlend {blendInTime=0.5, blendOutTime=0.5}
    boneToModify {boneName=dyn_hair_curl_02_cr}
    rotation=<rot 8.6789 -30.375 -5.557>
   }
   n10 = "Transform (Modify) Bone - Bone: offset_hair_cr"!(ComponentPose=@n9, bAlphaBoolEnabled:bool=True, Translation="0,0,0", Rotation="10,0,0", Scale="1,1,1") -> (Pose) {
    alphaInputType=Bool
    rotationMode=BMM_Additive
    rotationSpace=BCS_BoneSpace
    alphaBoolBlend {blendInTime=0.5, blendOutTime=0.5}
    boneToModify {boneName=offset_hair_cr}
    rotation {pitch=10}
   }
   Save cached pose 'AnimDynPose'(Pose=Component To Local(ComponentPose=@n10)) {cacheName=AnimDynPose}
   n11 = RigidBody!(
    ComponentPose=Local To Component(LocalPose=Use cached pose 'PreDynamicsPose'())
    Alpha
    OverrideWorldGravity:Vector="0,0,-980"
    ExternalForce
   ) -> (Pose) {
    bClampLinearTranslationLimitToRefPose=true
    bOverrideWorldGravity=true
    overridePhysicsAsset=&/F_MED_SaltLattice_FaceAcc_base_Physics
    simulationSpace=BaseBoneSpace
    alphaBoolBlend {blendOption=Linear}
    baseBoneRef {boneName=neck_02}
    overrideWorldGravity {z=-980}
    simSpaceSettings {
     dampingAlpha=0
     maxAngularAcceleration=9
     maxAngularVelocity=4
     maxLinearAcceleration=1000
     maxLinearVelocity=400
     worldAlpha=1
     externalLinearDragV=<1.1 1.1 1.1>
    }
   }
   Save cached pose 'BackHairPose'(Pose=Component To Local(ComponentPose=@n11)) {cacheName=BackHairPose}
   n12 = "Transform (Modify) Bone - Bone: offset_hair_outer_lf"!(
    ComponentPose=Local To Component(LocalPose=Use cached pose 'ControlRigPose'())
    bAlphaBoolEnabled:bool=True
    Translation="0,0,0"
    Rotation="19.491,9.207,3.084"
    Scale="1,1,1"
   ) -> (Pose) {
    alphaInputType=Bool
    rotationMode=BMM_Additive
    rotationSpace=BCS_BoneSpace
    boneToModify {boneName=offset_hair_outer_lf}
    rotation=<rot 19.491 9.207 3.084>
   }
   n13 = "Transform (Modify) Bone - Bone: offset_hair_outer_rt"!(ComponentPose=@n12, bAlphaBoolEnabled:bool=True, Translation="0,0,0", Rotation="12.852,-6.169,-1.432", Scale="1,1,1") -> (Pose) {
    alphaInputType=Bool
    rotationMode=BMM_Additive
    rotationSpace=BCS_BoneSpace
    boneToModify {boneName=offset_hair_outer_rt}
    rotation=<rot 12.852 -6.169 -1.432>
   }
   n14 = "Transform (Modify) Bone - Bone: offset_hair_lf"!(ComponentPose=@n13, bAlphaBoolEnabled:bool=True, Translation="0,0,0", Rotation="0,-10,0", Scale="1,1,1") -> (Pose) {alphaInputType=Bool, rotationMode=BMM_Additive, rotationSpace=BCS_BoneSpace, boneToModify {boneName=offset_hair_lf}, rotation {yaw=-10}}
   Save cached pose 'PreDynamicsPose'(Pose=Component To Local(ComponentPose=@n14)) {cacheName=PreDynamicsPose}
   n15 = Layered blend per bone(
    BasePose=Copy Pose From Mesh() {bUseAttachedParent=true}
    BlendPoses_0=Copy Pose From Mesh(SourceMeshComponent=Get Body_Mesh()) {bUseAttachedParent=true}
   ) {layerSetup [{branchFilters [{boneName=clavicle_l}, {boneName=clavicle_r}, {boneName=upperarm_r}, {boneName=upperarm_l}]}]}
   Save cached pose 'StartPse'(Pose=@n15) {cacheName=StartPse}
   n16 = Control Rig!(Source=Use cached pose 'StartPse'(), Alpha, swingPickAxe:bool=Get bIsPlayingMeleeAnim(), IsTargeting:bool=Get bIsTargeting()) -> (Pose) {controlRigAssetReference {blueprintRigClass=&/F_MED_SaltLattice_FaceAcc_CtrlRig.F_MED_SaltLattice_FaceAcc_CtrlRig_C}}
   Save cached pose 'ControlRigPose'(Pose=@n16) {cacheName=ControlRigPose}
   n17 = "Constraint - dyn_hair_curl_02_lf (PPPPPPPPP)"!(
    ComponentPose=Local To Component(LocalPose=Use cached pose 'BackHairPose'())
    Alpha=1
    ConstraintWeights_0=0.4291
    ConstraintWeights_1:float=0.2627
    ConstraintWeights_2:float=0.1403
    ConstraintWeights_3:float=0.105
    ConstraintWeights_4:float=0.0221
    ConstraintWeights_5:float=0.0153
    ConstraintWeights_6:float=0.0093
    ConstraintWeights_7:float=0.0084
    ConstraintWeights_8:float=0.0077
   ) -> (Pose) {
    boneToModify {boneName=dyn_hair_curl_02_lf}
    constraintSetup [
     {targetBone {boneName=dyn_hair_01_lf}}
     {targetBone {boneName=dyn_hair_02_lf}}
     {targetBone {boneName=dyn_hair_01_cr}}
     {targetBone {boneName=dyn_hair_02_cr}}
     {targetBone {boneName=dyn_hair_03_lf}}
     {targetBone {boneName=dyn_hair_01_rt}}
     {targetBone {boneName=dyn_hair_03_cr}}
     {targetBone {boneName=dyn_hair_02_rt}}
     {targetBone {boneName=head}}
    ]
    constraintWeights [1,1,1,1,1,1,1,1,1]
   }
   n18 = "Constraint - dyn_hair_curl_03_lf (PPPPPPPPP)"!(
    ComponentPose=@n17
    Alpha=1
    ConstraintWeights_0=0.3791
    ConstraintWeights_1:float=0.3452
    ConstraintWeights_2:float=0.12
    ConstraintWeights_3:float=0.0857
    ConstraintWeights_4:float=0.0385
    ConstraintWeights_5:float=0.0116
    ConstraintWeights_6:float=0.0108
    ConstraintWeights_7:float=0.008
    ConstraintWeights_8:float=0.001
   ) -> (Pose) {
    boneToModify {boneName=dyn_hair_curl_03_lf}
    constraintSetup [
     {targetBone {boneName=dyn_hair_03_lf}}
     {targetBone {boneName=dyn_hair_02_lf}}
     {targetBone {boneName=dyn_hair_03_cr}}
     {targetBone {boneName=dyn_hair_02_cr}}
     {targetBone {boneName=dyn_hair_01_lf}}
     {targetBone {boneName=dyn_hair_04_lf}}
     {targetBone {boneName=dyn_hair_01_cr}}
     {targetBone {boneName=head}}
     {targetBone {boneName=dyn_hair_02_rt}}
    ]
    constraintWeights [1,1,1,1,1,1,1,1,1]
   }
   n19 = "Constraint - dyn_hair_curl_04_lf (PPPPPPPPP)"!(
    ComponentPose=@n18
    Alpha=1
    ConstraintWeights_0=0.5253
    ConstraintWeights_1:float=0.2903
    ConstraintWeights_2:float=0.12
    ConstraintWeights_3:float=0.0412
    ConstraintWeights_4:float=0.0075
    ConstraintWeights_5:float=0.006
    ConstraintWeights_6:float=0.005
    ConstraintWeights_7:float=0.0033
    ConstraintWeights_8:float=0.0014
   ) -> (Pose) {
    boneToModify {boneName=dyn_hair_curl_04_lf}
    constraintSetup [
     {targetBone {boneName=dyn_hair_04_lf}}
     {targetBone {boneName=dyn_hair_03_lf}}
     {targetBone {boneName=dyn_hair_04_cr}}
     {targetBone {boneName=dyn_hair_03_cr}}
     {targetBone {boneName=dyn_hair_02_lf}}
     {targetBone {boneName=dyn_hair_05_lf}}
     {targetBone {boneName=head}}
     {targetBone {boneName=dyn_hair_05_cr}}
     {targetBone {boneName=dyn_hair_02_cr}}
    ]
    constraintWeights [1,1,1,1,1,1,1,1,1]
   }
   n20 = "Constraint - dyn_hair_curl_05_lf (PPPPPPP)"!(
    ComponentPose=@n19
    Alpha=1
    ConstraintWeights_0=0.5659
    ConstraintWeights_1:float=0.203
    ConstraintWeights_2:float=0.1274
    ConstraintWeights_3:float=0.1007
    ConstraintWeights_4:float=0.0022
    ConstraintWeights_5:float=0.0006
    ConstraintWeights_6:float=0.0001
   ) -> (Pose) {
    boneToModify {boneName=dyn_hair_curl_05_lf}
    constraintSetup [
     {targetBone {boneName=dyn_hair_04_lf}}
     {targetBone {boneName=dyn_hair_04_cr}}
     {targetBone {boneName=dyn_hair_05_lf}}
     {targetBone {boneName=dyn_hair_05_cr}}
     {targetBone {boneName=dyn_hair_03_lf}}
     {targetBone {boneName=dyn_hair_03_cr}}
     {targetBone {boneName=dyn_hair_05_rt}}
    ]
    constraintWeights [1,1,1,1,1,1,1]
   }
   n21 = "Constraint - dyn_hair_curl_02_cr (PPPPPPPPP)"!(
    ComponentPose=@n20
    Alpha=1
    ConstraintWeights_0=0.4088
    ConstraintWeights_1:float=0.3686
    ConstraintWeights_2:float=0.0723
    ConstraintWeights_3:float=0.067
    ConstraintWeights_4:float=0.0278
    ConstraintWeights_5:float=0.0255
    ConstraintWeights_6:float=0.0137
    ConstraintWeights_7:float=0.0116
    ConstraintWeights_8:float=0.0048
   ) -> (Pose) {
    boneToModify {boneName=dyn_hair_curl_02_cr}
    constraintSetup [
     {targetBone {boneName=dyn_hair_03_cr}}
     {targetBone {boneName=dyn_hair_02_cr}}
     {targetBone {boneName=dyn_hair_03_lf}}
     {targetBone {boneName=dyn_hair_02_lf}}
     {targetBone {boneName=dyn_hair_01_cr}}
     {targetBone {boneName=dyn_hair_02_rt}}
     {targetBone {boneName=dyn_hair_03_rt}}
     {targetBone {boneName=head}}
     {targetBone {boneName=dyn_hair_01_lf}}
    ]
    constraintWeights [1,1,1,1,1,1,1,1,1]
   }
   n22 = "Constraint - dyn_hair_curl_03_cr (PPPPPPPP)"!(
    ComponentPose=@n21
    Alpha=1
    ConstraintWeights_0=0.4598
    ConstraintWeights_1:float=0.265
    ConstraintWeights_2:float=0.1619
    ConstraintWeights_3:float=0.1088
    ConstraintWeights_4:float=0.0031
    ConstraintWeights_5:float=0.0008
    ConstraintWeights_6:float=0.0004
    ConstraintWeights_7:float=0.0003
   ) -> (Pose) {
    boneToModify {boneName=dyn_hair_curl_03_cr}
    constraintSetup [
     {targetBone {boneName=dyn_hair_03_cr}}
     {targetBone {boneName=dyn_hair_04_cr}}
     {targetBone {boneName=dyn_hair_03_lf}}
     {targetBone {boneName=dyn_hair_04_lf}}
     {targetBone {boneName=dyn_hair_02_cr}}
     {targetBone {boneName=dyn_hair_02_lf}}
     {targetBone {boneName=dyn_hair_03_rt}}
     {targetBone {boneName=dyn_hair_05_lf}}
    ]
    constraintWeights [1,1,1,1,1,1,1,1]
   }
   n23 = "Constraint - dyn_hair_curl_04_cr (PPPPPP)"!(
    ComponentPose=@n22
    Alpha=1
    ConstraintWeights_0=0.7435
    ConstraintWeights_1:float=0.1667
    ConstraintWeights_2:float=0.0413
    ConstraintWeights_3:float=0.0251
    ConstraintWeights_4:float=0.0205
    ConstraintWeights_5:float=0.0029
   ) -> (Pose) {
    boneToModify {boneName=dyn_hair_curl_04_cr}
    constraintSetup [
     {targetBone {boneName=dyn_hair_04_lf}}
     {targetBone {boneName=dyn_hair_04_cr}}
     {targetBone {boneName=dyn_hair_05_lf}}
     {targetBone {boneName=dyn_hair_05_cr}}
     {targetBone {boneName=dyn_hair_03_lf}}
     {targetBone {boneName=dyn_hair_03_cr}}
    ]
    constraintWeights [1,1,1,1,1,1]
   }
   n24 = "Constraint - dyn_hair_curl_02_rt (PPPPPPPP)"!(
    ComponentPose=@n23
    Alpha=1
    ConstraintWeights_0=0.4897
    ConstraintWeights_1:float=0.3117
    ConstraintWeights_2:float=0.0734
    ConstraintWeights_3:float=0.058
    ConstraintWeights_4:float=0.0353
    ConstraintWeights_5:float=0.0221
    ConstraintWeights_6:float=0.0074
    ConstraintWeights_7:float=0.0024
   ) -> (Pose) {
    boneToModify {boneName=dyn_hair_curl_02_rt}
    constraintSetup [
     {targetBone {boneName=dyn_hair_03_rt}}
     {targetBone {boneName=dyn_hair_02_rt}}
     {targetBone {boneName=head}}
     {targetBone {boneName=dyn_hair_02_cr}}
     {targetBone {boneName=dyn_hair_01_rt}}
     {targetBone {boneName=dyn_hair_03_cr}}
     {targetBone {boneName=dyn_hair_04_rt}}
     {targetBone {boneName=dyn_hair_01_cr}}
    ]
    constraintWeights [1,1,1,1,1,1,1,1]
   }
   n25 = "Constraint - dyn_hair_curl_03_rt (PPPPPPPP)"!(
    ComponentPose=@n24
    Alpha=1
    ConstraintWeights_0=0.7494
    ConstraintWeights_1:float=0.1514
    ConstraintWeights_2:float=0.0783
    ConstraintWeights_3:float=0.0153
    ConstraintWeights_4:float=0.0024
    ConstraintWeights_5:float=0.0015
    ConstraintWeights_6:float=0.0012
    ConstraintWeights_7:float=0.0004
   ) -> (Pose) {
    boneToModify {boneName=dyn_hair_curl_03_rt}
    constraintSetup [
     {targetBone {boneName=dyn_hair_03_rt}}
     {targetBone {boneName=dyn_hair_04_rt}}
     {targetBone {boneName=head}}
     {targetBone {boneName=dyn_hair_02_rt}}
     {targetBone {boneName=dyn_hair_03_cr}}
     {targetBone {boneName=dyn_hair_02_cr}}
     {targetBone {boneName=dyn_hair_04_cr}}
     {targetBone {boneName=dyn_hair_01_rt}}
    ]
    constraintWeights [1,1,1,1,1,1,1,1]
   }
   n26 = "Constraint - dyn_hair_curl_04_rt (PPPPP)"!(
    ComponentPose=@n25
    Alpha=1
    ConstraintWeights_0=0.9428
    ConstraintWeights_1:float=0.0313
    ConstraintWeights_2:float=0.0196
    ConstraintWeights_3:float=0.0049
    ConstraintWeights_4:float=0.0014
   ) -> (Pose) {
    boneToModify {boneName=dyn_hair_curl_04_rt}
    constraintSetup [
     {targetBone {boneName=dyn_hair_04_rt}}
     {targetBone {boneName=dyn_hair_05_rt}}
     {targetBone {boneName=dyn_hair_04_cr}}
     {targetBone {boneName=dyn_hair_03_rt}}
     {targetBone {boneName=dyn_hair_05_cr}}
    ]
    constraintWeights [1,1,1,1,1]
   }
   n27 = "Transform (Modify) Bone - Bone: offset_hair_lf"(ComponentPose=@n26, bAlphaBoolEnabled=True, Translation="-2,0,0", Rotation="0,-10,0", Scale="1,1,1") {
    alphaInputType=Bool
    rotationMode=BMM_Additive
    rotationSpace=BCS_BoneSpace
    translationMode=BMM_Additive
    translationSpace=BCS_BoneSpace
    alphaBoolBlend {blendInTime=0.5, blendOutTime=0.5}
    boneToModify {boneName=offset_hair_lf}
    rotation {yaw=-10}
    translation {x=-2}
   }
   n28 = "Transform (Modify) Bone - Bone: offset_hair_rt"!(ComponentPose=@n27, bAlphaBoolEnabled:bool=True, Translation="-2,0,0", Rotation="0,-10,0", Scale="1,1,1") -> (Pose) {
    alphaInputType=Bool
    rotationMode=BMM_Additive
    rotationSpace=BCS_BoneSpace
    translationMode=BMM_Additive
    translationSpace=BCS_BoneSpace
    alphaBoolBlend {blendInTime=0.5, blendOutTime=0.5}
    boneToModify {boneName=offset_hair_rt}
    rotation {yaw=-10}
    translation {x=-2}
   }
   n29 = "Transform (Modify) Bone - Bone: offset_hair_cr"(ComponentPose=@n28, bAlphaBoolEnabled=True, Translation="-2,0,0", Rotation="0,-10,0", Scale="1,1,1") {
    alphaInputType=Bool
    rotationMode=BMM_Additive
    rotationSpace=BCS_BoneSpace
    translationMode=BMM_Additive
    translationSpace=BCS_BoneSpace
    alphaBoolBlend {blendInTime=0.5, blendOutTime=0.5}
    boneToModify {boneName=offset_hair_cr}
    rotation {yaw=-10}
    translation {x=-2}
   }
   n30 = "Constraint - dyn_hair_02_lf (P)"!(ComponentPose=@n29, bAlphaBoolEnabled:bool=True, ConstraintWeights_0=1) -> (Pose) {alphaInputType=Bool, boneToModify {boneName=dyn_hair_02_lf}, constraintSetup [{targetBone {boneName=spine_05}}], constraintWeights [1]}
   n31 = "Constraint - dyn_hair_03_lf (P)"!(ComponentPose=@n30, bAlphaBoolEnabled:bool=True, ConstraintWeights_0=1) -> (Pose) {alphaInputType=Bool, boneToModify {boneName=dyn_hair_03_lf}, constraintSetup [{targetBone {boneName=spine_05}}], constraintWeights [1]}
   n32 = "Constraint - dyn_hair_04_lf (P)"!(ComponentPose=@n31, bAlphaBoolEnabled:bool=True, ConstraintWeights_0=1) -> (Pose) {alphaInputType=Bool, boneToModify {boneName=dyn_hair_04_lf}, constraintSetup [{targetBone {boneName=spine_05}}], constraintWeights [1]}
   n33 = "Constraint - dyn_hair_05_lf (P)"!(ComponentPose=@n32, bAlphaBoolEnabled:bool=True, ConstraintWeights_0=1) -> (Pose) {alphaInputType=Bool, boneToModify {boneName=dyn_hair_05_lf}, constraintSetup [{targetBone {boneName=spine_05}}], constraintWeights [1]}
   n34 = "Constraint - dyn_hair_01_cr (RR)"!(ComponentPose=@n33, bAlphaBoolEnabled:bool=True, ConstraintWeights_0=0.5, ConstraintWeights_1:float=0.5) -> (Pose) {
    alphaInputType=Bool
    boneToModify {boneName=dyn_hair_01_cr}
    constraintSetup [{transformType=Rotation, targetBone {boneName=head}}, {transformType=Rotation, targetBone {boneName=spine_05}}]
    constraintWeights [1,1]
   }
   n35 = "Constraint - dyn_hair_02_cr (P)"!(ComponentPose=@n34, bAlphaBoolEnabled:bool=True, ConstraintWeights_0=1) -> (Pose) {alphaInputType=Bool, boneToModify {boneName=dyn_hair_02_cr}, constraintSetup [{targetBone {boneName=spine_05}}], constraintWeights [1]}
   n36 = "Constraint - dyn_hair_03_cr (P)"!(ComponentPose=@n35, bAlphaBoolEnabled:bool=True, ConstraintWeights_0=1) -> (Pose) {alphaInputType=Bool, boneToModify {boneName=dyn_hair_03_cr}, constraintSetup [{targetBone {boneName=spine_05}}], constraintWeights [1]}
   n37 = "Constraint - dyn_hair_04_cr (P)"!(ComponentPose=@n36, bAlphaBoolEnabled:bool=True, ConstraintWeights_0=1) -> (Pose) {alphaInputType=Bool, boneToModify {boneName=dyn_hair_04_cr}, constraintSetup [{targetBone {boneName=spine_05}}], constraintWeights [1]}
   n38 = "Constraint - dyn_hair_05_cr (P)"!(ComponentPose=@n37, bAlphaBoolEnabled:bool=True, ConstraintWeights_0=1) -> (Pose) {alphaInputType=Bool, boneToModify {boneName=dyn_hair_05_cr}, constraintSetup [{targetBone {boneName=spine_05}}], constraintWeights [1]}
   n39 = "Constraint - dyn_hair_01_rt (R)"!(ComponentPose=@n38, bAlphaBoolEnabled:bool=True, ConstraintWeights_0=1) -> (Pose) {
    alphaInputType=Bool
    boneToModify {boneName=dyn_hair_01_rt}
    constraintSetup [{transformType=Rotation, targetBone {boneName=head}}]
    constraintWeights [1]
   }
   n40 = "Constraint - dyn_hair_02_rt (P)"!(ComponentPose=@n39, bAlphaBoolEnabled:bool=True, ConstraintWeights_0=1) -> (Pose) {alphaInputType=Bool, boneToModify {boneName=dyn_hair_02_rt}, constraintSetup [{targetBone {boneName=spine_05}}], constraintWeights [1]}
   n41 = "Constraint - dyn_hair_03_rt (P)"!(ComponentPose=@n40, bAlphaBoolEnabled:bool=True, ConstraintWeights_0=1) -> (Pose) {alphaInputType=Bool, boneToModify {boneName=dyn_hair_03_rt}, constraintSetup [{targetBone {boneName=spine_05}}], constraintWeights [1]}
   n42 = "Constraint - dyn_hair_04_rt (P)"!(ComponentPose=@n41, bAlphaBoolEnabled:bool=True, ConstraintWeights_0=1) -> (Pose) {alphaInputType=Bool, boneToModify {boneName=dyn_hair_04_rt}, constraintSetup [{targetBone {boneName=spine_05}}], constraintWeights [1]}
   n43 = "Constraint - dyn_hair_04_rt (P)"(ComponentPose=@n42, bAlphaBoolEnabled=True, ConstraintWeights_0=1) {alphaInputType=Bool, boneToModify {boneName=dyn_hair_04_rt}, constraintSetup [{targetBone {boneName=spine_05}}], constraintWeights [1]}
   n44 = "Constraint - dyn_hair_curl_04_cr (P)"!(ComponentPose=@n43, bAlphaBoolEnabled:bool=True, ConstraintWeights_0=1) -> (Pose) {alphaInputType=Bool, boneToModify {boneName=dyn_hair_curl_04_cr}, constraintSetup [{targetBone {boneName=spine_05}}], constraintWeights [1]}
   n45 = "Constraint - dyn_hair_curl_03_cr (P)"!(ComponentPose=@n44, bAlphaBoolEnabled:bool=True, ConstraintWeights_0=1) -> (Pose) {alphaInputType=Bool, boneToModify {boneName=dyn_hair_curl_03_cr}, constraintSetup [{targetBone {boneName=spine_05}}], constraintWeights [1]}
   n46 = "Constraint - dyn_hair_curl_03_rt (P)"!(ComponentPose=@n45, bAlphaBoolEnabled:bool=True, ConstraintWeights_0=1) -> (Pose) {alphaInputType=Bool, boneToModify {boneName=dyn_hair_curl_03_rt}, constraintSetup [{targetBone {boneName=spine_05}}], constraintWeights [1]}
   n47 = "Constraint - dyn_hair_curl_04_rt (P)"!(ComponentPose=@n46, bAlphaBoolEnabled:bool=True, ConstraintWeights_0=1) -> (Pose) {alphaInputType=Bool, boneToModify {boneName=dyn_hair_curl_04_rt}, constraintSetup [{targetBone {boneName=spine_05}}], constraintWeights [1]}
   n48 = "Constraint - dyn_hair_curl_03_lf (P)"!(ComponentPose=@n47, bAlphaBoolEnabled:bool=True, ConstraintWeights_0=1) -> (Pose) {alphaInputType=Bool, boneToModify {boneName=dyn_hair_curl_03_lf}, constraintSetup [{targetBone {boneName=spine_05}}], constraintWeights [1]}
   n49 = "Constraint - dyn_hair_curl_04_lf (P)"!(ComponentPose=@n48, bAlphaBoolEnabled:bool=True, ConstraintWeights_0=1) -> (Pose) {alphaInputType=Bool, boneToModify {boneName=dyn_hair_curl_04_lf}, constraintSetup [{targetBone {boneName=spine_05}}], constraintWeights [1]}
   n50 = "Constraint - dyn_hair_curl_05_lf (P)"!(ComponentPose=@n49, bAlphaBoolEnabled:bool=True, ConstraintWeights_0=1) -> (Pose) {alphaInputType=Bool, boneToModify {boneName=dyn_hair_curl_05_lf}, constraintSetup [{targetBone {boneName=spine_05}}], constraintWeights [1]}
   n51 = "Transform (Modify) Bone - Bone: offset_hair_rt"(ComponentPose=@n50, bAlphaBoolEnabled=True, Translation="0,0,0", Rotation="8.034,2,0.067", Scale="1,1,1") {
    alphaInputType=Bool
    rotationMode=BMM_Additive
    rotationSpace=BCS_BoneSpace
    translationMode=BMM_Additive
    translationSpace=BCS_BoneSpace
    alphaBoolBlend {blendInTime=0.5, blendOutTime=0.5}
    boneToModify {boneName=offset_hair_rt}
    rotation=<rot 8.034 2 0.067>
   }
   n52 = "Transform (Modify) Bone - Bone: offset_hair_lf"(ComponentPose=@n51, bAlphaBoolEnabled=True, Translation="0,0,0", Rotation="5,-5,0", Scale="1,1,1") {
    alphaInputType=Bool
    rotationMode=BMM_Additive
    rotationSpace=BCS_BoneSpace
    translationMode=BMM_Additive
    translationSpace=BCS_BoneSpace
    alphaBoolBlend {blendInTime=0.5, blendOutTime=0.5}
    boneToModify {boneName=offset_hair_lf}
    rotation {pitch=5, yaw=-5}
   }
   n53 = Control Rig!(Source=Component To Local(ComponentPose=@n52), Alpha) -> (Pose) {controlRigAssetReference {blueprintRigClass=&/F_MED_SaltLattice_FaceAcc_postSim_CtrlRig.F_MED_SaltLattice_FaceAcc_postSim_CtrlRig_C}}
   Save cached pose 'ConstraintPose'(Pose=@n53) {cacheName=ConstraintPose}
  ]}
 EventGraph {nodes [
   Event OnCharacterCustomizationLoaded()
   > n1 = Cast To PlayerPawn_Athena(Object=TryGetPawnOwner())
   > Set Body_Mesh(Body_Mesh=GetSkeletalMeshForPartType(self=@n1))
   n2 = CheckForBackbling()
   > n3 = Set isWearingBackbling(isWearingBackbling=AND Boolean!(A=@n2, B=NOT Boolean(A=@n2.bIsBackblingACape), C:bool=NOT Boolean(A=@n2.bIsBackblingLong)) -> (ReturnValue))
   > n4 = Set Is Wearing Cape(Is Wearing Cape=OR Boolean(A=@n2.bIsBackblingACape, B=@n2.bIsBackblingLong))
   > Set is Wearing cape and backpack(is Wearing cape and backpack:bool=OR Boolean(A=@n3, B=@n4)) -> (Output_Get:bool)
   Event BlueprintUpdateAnimation()
   > n5 = Sequence!(execute) -> (then_0, then_1, then_2:Exec)
   > n6 = "Gravity Override State Function (Parameter Asset) (Nativized)"(
    CharacterPartAnimInstance=Self-Reference()
    GameTimeInSeconds=Get Game Time in Seconds() -> (Game Time in Seconds:float)
    Parameters=&/F_MED_SaltLattice_LHair_Parameters
   )
   > Set Gravity Override(Gravity Override=@n6)
   n7 = Get is Wearing cape and backpack()
   Set isSkydivingWithoutBackpack(execute=@n5.then_2, isSkydivingWithoutBackpack:bool=AND Boolean!(A=NOT Boolean(A=@n7), B=Get bIsSkydiveFloating()) -> (ReturnValue)) -> (Output_Get:bool)
   > Set IsFloatingWithBackpack(IsFloatingWithBackpack:bool=AND Boolean(A=Get is Wearing cape and backpack(), B=Get bIsSkydiveFloating())) -> (Output_Get:bool)
   n8 = GetOwnerAnimationInterface!(self:Custom Character Part Anim Instance Object Reference) -> (ReturnValue:Fort Player Animation Interface Object Reference)
   Is Valid(exec=@n5.then_1, InputObject=@n8)
   > n9 = Branch(Condition=Break Fort Anim Input Zipline!(FortAnimInput_Zipline:"Fort Anim Input Zipline Structure (by ref)"=GetZiplineInput!(self:Fort Player Animation Interface Object Reference=@n8) -> (ReturnValue:"Fort Anim Input Zipline Structure (by ref)")) -> (
     bIsZiplining:bool=~
     bIsOnAscender:bool=~
     bShouldPlayPivotTransition:bool=~
     LeanYaw:float=~
     PivotBlendDelayRemaining:float=~
     PivotBlendDelay:float=~
     PivotCardinalDirection:EFortCardinalDirection Enum
     WorldVelocityLastTick:Vector=~
    ) {is=title K2Node_BreakStruct})
   > Set Componen Linear Vel Scale(Componen Linear Vel Scale:Vector="0, 0, 0") -> (Output_Get:Vector)
   Set Componen Linear Vel Scale(execute=@n9.else, Componen Linear Vel Scale="0.3,0.3,0.3")
  ]}
]
variables [
 Body_Mesh {type=SkeletalMeshComponent}
 Componen Linear Vel Scale {type=Vector, default=<0.3 0.3 0.3>}
 Game Time in Seconds {type=number}
 Gravity Override {type=Vector}
 HairAlpha {type=number, default=0.3}
 Is Wearing Cape {type=boolean}
 is Wearing cape and backpack {type=boolean}
 IsCrouchingTrue {type=boolean}
 IsFloatingWithBackpack {type=boolean}
 isSkydivingWithoutBackpack {type=boolean}
 isWearingBackbling {type=boolean}
 IsWearingCape_Not {type=boolean}
]
