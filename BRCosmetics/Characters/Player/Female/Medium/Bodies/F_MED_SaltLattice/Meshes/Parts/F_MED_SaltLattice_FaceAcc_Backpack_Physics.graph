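# Physics asset for the SaltLattice medium female FaceAcc Backpack mesh part, set up for dangling hair bone chains.
# Three back-hair chains and two side-hair chains of identical capsules hang from small base bodies on the head, with side chains sharing the left and right bases.
# Chain joints have zero twist limit and narrow swing limits that widen toward the tips.
# Spine, neck, clavicle and head capsules form the upper-body volumes, joined by free angular constraints.
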PhysicsAsset /BRCosmetics/Characters/Player/Female/Medium/Bodies/F_MED_SaltLattice/Meshes/Parts/F_MED_SaltLattice_FaceAcc_Backpack_Physics
bodies {
 base_backHair_cr [{shapeType=Capsule, length=0.505, radius=0.505, shapeName=base_backHair_cr_capsule, rotation=<rot 0 -90 0>}]
 base_backHair_lf [{shapeType=Capsule, length=0.505, radius=0.505, shapeName=base_backHair_lf_capsule, rotation=<rot 0 -90 0>}]
 base_backHair_rt [{shapeType=Capsule, length=0.505, radius=0.505, shapeName=base_backHair_rt_capsule, rotation=<rot 0 -90 0>}]
 clavicle_l [{shapeType=Capsule, length=4.9228, radius=5.7886, shapeName=clavicle_l_capsule, center=<6.3703 -1.5625 0.0251>, rotation=<rot 90 -179.9999 0>}]
 clavicle_r [{
   shapeType=Capsule
   length=4.7873
   radius=5.9241
   shapeName=clavicle_r_capsule
   center=<-6.331 1.3506 0.0875>
   rotation=<rot 85 -0.0002 -0.0004>
  }]
 dyn_backHair_01_cr [{shapeType=Capsule, length=8, radius=2, center=<8 0 1.1>, rotation=<rot 90 0 0>}]
 dyn_backHair_01_lf [{shapeType=Capsule, length=8, radius=2, center=<8 0 1.1>, rotation=<rot 90 0 0>}]
 dyn_backHair_01_rt [{shapeType=Capsule, length=8, radius=2, center=<8 0 1.1>, rotation=<rot 90 0 0>}]
 dyn_backHair_02_cr [{shapeType=Capsule, length=8, radius=2, center=<8 0 1.1>, rotation=<rot 90 0 0>}]
 dyn_backHair_02_lf [{shapeType=Capsule, length=8, radius=2, center=<8 0 1.1>, rotation=<rot 90 0 0>}]
 dyn_backHair_02_rt [{shapeType=Capsule, length=8, radius=2, center=<8 0 1.1>, rotation=<rot 90 0 0>}]
 dyn_backHair_03_cr [{shapeType=Capsule, length=8, radius=2, center=<8 0 1.1>, rotation=<rot 90 0 0>}]
 dyn_backHair_03_lf [{shapeType=Capsule, length=8, radius=2, center=<8 0 1.1>, rotation=<rot 90 0 0>}]
 dyn_backHair_03_rt [{shapeType=Capsule, length=8, radius=2, center=<8 0 1.1>, rotation=<rot 90 0 0>}]
 dyn_backHair_04_cr [{shapeType=Capsule, length=8, radius=2, center=<8 0 1.1>, rotation=<rot 90 0 0>}]
 dyn_backHair_04_lf [{shapeType=Capsule, length=8, radius=2, center=<8 0 1.1>, rotation=<rot 90 0 0>}]
 dyn_backHair_04_rt [{shapeType=Capsule, length=8, radius=2, center=<8 0 1.1>, rotation=<rot 90 0 0>}]
 dyn_backHair_05_cr_end [{shapeType=Capsule, length=8, radius=2, center=<8 0 1.1>, rotation=<rot 90 0 0>}]
 dyn_sideHair_b_01_lf [{shapeType=Capsule, length=8, radius=2, center=<8 0 0>, rotation=<rot 90 0 0>}]
 dyn_sideHair_b_01_rt [{shapeType=Capsule, length=8, radius=2, center=<8 0 0>, rotation=<rot 90 0 0>}]
 dyn_sideHair_b_02_lf [{shapeType=Capsule, length=8, radius=2, center=<8 0 0>, rotation=<rot 90 0 0>}]
 dyn_sideHair_b_02_rt [{shapeType=Capsule, length=8, radius=2, center=<8 0 0>, rotation=<rot 90 0 0>}]
 dyn_sideHair_b_03_lf [{shapeType=Capsule, length=8, radius=2, center=<8 0 0>, rotation=<rot 90 0 0>}]
 dyn_sideHair_b_03_rt [{shapeType=Capsule, length=8, radius=2, center=<8 0 0>, rotation=<rot 90 0 0>}]
 dyn_sideHair_b_04_lf [{shapeType=Capsule, length=8, radius=2, center=<8 0 0>, rotation=<rot 90 0 0>}]
 dyn_sideHair_b_04_rt [{shapeType=Capsule, length=8, radius=2, center=<8 0 0>, rotation=<rot 90 0 0>}]
 head [{shapeType=Capsule, length=1, radius=1}]
 neck_02 [{shapeType=Capsule, length=18.4237, radius=12.3308, shapeName=neck_02_capsule, center=<3.5779 12.2787 0>, rotation=<rot 0 -90 0>}]
 spine_04 [
  {shapeType=Capsule, length=19, radius=8, shapeName=spine_05_capsule, center=<-9.7587 0.1497 0>, rotation=<rot 0 -90 0>}
  {shapeType=Capsule, length=20, radius=8, shapeName=spine_04_capsule, center=<-0.068 -2.6126 0>}
 ]
 spine_05 [
  {shapeType=Capsule, length=19, radius=9, shapeName=spine_05_capsule, center=<2.2882 -1.5376 0>, rotation=<rot 0 -90 0>}
  {shapeType=Capsule, length=25.5089, radius=20.0964, shapeName=spine_05_capsule, center=<16.299 -30.6292 0>}
 ]
}
constraints [
 {
  bone1Name=neck_02
  bone2Name=spine_05
  swing1LimitDegrees=45
  swing1Motion=Free
  swing2LimitDegrees=45
  swing2Motion=Free
  twistLimitDegrees=45
  twistMotion=Free
 }
 {
  bone1Name=head
  bone2Name=neck_02
  swing1LimitDegrees=45
  swing1Motion=Free
  swing2LimitDegrees=45
  swing2Motion=Free
  twistLimitDegrees=45
  twistMotion=Free
 }
 {
  bone1Name=clavicle_l
  bone2Name=spine_05
  swing1LimitDegrees=45
  swing1Motion=Free
  swing2LimitDegrees=45
  swing2Motion=Free
  twistLimitDegrees=45
  twistMotion=Free
 }
 {
  bone1Name=clavicle_r
  bone2Name=spine_05
  swing1LimitDegrees=45
  swing1Motion=Free
  swing2LimitDegrees=45
  swing2Motion=Free
  twistLimitDegrees=45
  twistMotion=Free
 }
 {
  bone1Name=spine_05
  bone2Name=spine_04
  swing1LimitDegrees=45
  swing1Motion=Free
  swing2LimitDegrees=45
  swing2Motion=Free
  twistLimitDegrees=45
  twistMotion=Free
 }
 {
  bone1Name=base_backHair_cr
  bone2Name=head
  swing1LimitDegrees=45
  swing1Motion=Free
  swing2LimitDegrees=45
  swing2Motion=Free
  twistLimitDegrees=45
  twistMotion=Free
 }
 {
  bone1Name=base_backHair_lf
  bone2Name=head
  swing1LimitDegrees=45
  swing1Motion=Free
  swing2LimitDegrees=45
  swing2Motion=Free
  twistLimitDegrees=45
  twistMotion=Free
 }
 {
  bone1Name=base_backHair_rt
  bone2Name=head
  swing1LimitDegrees=45
  swing1Motion=Free
  swing2LimitDegrees=45
  swing2Motion=Free
  twistLimitDegrees=45
  twistMotion=Free
 }
 {bone1Name=dyn_backHair_01_cr, bone2Name=base_backHair_cr, swing1LimitDegrees=0, swing2LimitDegrees=5, twistLimitDegrees=0}
 {bone1Name=dyn_backHair_02_cr, bone2Name=dyn_backHair_01_cr, swing1LimitDegrees=5, swing2LimitDegrees=10, twistLimitDegrees=0}
 {bone1Name=dyn_backHair_03_cr, bone2Name=dyn_backHair_02_cr, swing1LimitDegrees=10, swing2LimitDegrees=10, twistLimitDegrees=0}
 {bone1Name=dyn_backHair_04_cr, bone2Name=dyn_backHair_03_cr, swing1LimitDegrees=10, swing2LimitDegrees=10, twistLimitDegrees=0}
 {bone1Name=dyn_backHair_05_cr_end, bone2Name=dyn_backHair_04_cr, swing1LimitDegrees=10, swing2LimitDegrees=10, twistLimitDegrees=0}
 {bone1Name=dyn_backHair_01_lf, bone2Name=base_backHair_lf, swing1LimitDegrees=0, swing2LimitDegrees=5, twistLimitDegrees=0}
 {bone1Name=dyn_backHair_02_lf, bone2Name=dyn_backHair_01_lf, swing1LimitDegrees=5, swing2LimitDegrees=10, twistLimitDegrees=0}
 {bone1Name=dyn_backHair_03_lf, bone2Name=dyn_backHair_02_lf, swing1LimitDegrees=10, swing2LimitDegrees=10, twistLimitDegrees=0}
 {bone1Name=dyn_backHair_04_lf, bone2Name=dyn_backHair_03_lf, swing1LimitDegrees=10, swing2LimitDegrees=10, twistLimitDegrees=0}
 {bone1Name=dyn_backHair_01_rt, bone2Name=base_backHair_rt, swing1LimitDegrees=0, swing2LimitDegrees=5, twistLimitDegrees=0}
 {bone1Name=dyn_backHair_02_rt, bone2Name=dyn_backHair_01_rt, swing1LimitDegrees=5, swing2LimitDegrees=10, twistLimitDegrees=0}
 {bone1Name=dyn_backHair_03_rt, bone2Name=dyn_backHair_02_rt, swing1LimitDegrees=10, swing2LimitDegrees=10, twistLimitDegrees=0}
 {bone1Name=dyn_backHair_04_rt, bone2Name=dyn_backHair_03_rt, swing1LimitDegrees=10, swing2LimitDegrees=10, twistLimitDegrees=0}
 {bone1Name=dyn_sideHair_b_01_lf, bone2Name=base_backHair_lf, swing1LimitDegrees=0, swing2LimitDegrees=5, twistLimitDegrees=0}
 {bone1Name=dyn_sideHair_b_02_lf, bone2Name=dyn_sideHair_b_01_lf, swing1LimitDegrees=5, swing2LimitDegrees=10, twistLimitDegrees=0}
 {bone1Name=dyn_sideHair_b_03_lf, bone2Name=dyn_sideHair_b_02_lf, swing1LimitDegrees=10, swing2LimitDegrees=10, twistLimitDegrees=0}
 {bone1Name=dyn_sideHair_b_04_lf, bone2Name=dyn_sideHair_b_03_lf, swing1LimitDegrees=10, swing2LimitDegrees=10, twistLimitDegrees=0}
 {bone1Name=dyn_sideHair_b_01_rt, bone2Name=base_backHair_rt, swing1LimitDegrees=0, swing2LimitDegrees=5, twistLimitDegrees=0}
 {bone1Name=dyn_sideHair_b_02_rt, bone2Name=dyn_sideHair_b_01_rt, swing1LimitDegrees=5, swing2LimitDegrees=10, twistLimitDegrees=0}
 {bone1Name=dyn_sideHair_b_03_rt, bone2Name=dyn_sideHair_b_02_rt, swing1LimitDegrees=10, swing2LimitDegrees=10, twistLimitDegrees=0}
 {bone1Name=dyn_sideHair_b_04_rt, bone2Name=dyn_sideHair_b_03_rt, swing1LimitDegrees=10, swing2LimitDegrees=10, twistLimitDegrees=0}
]
